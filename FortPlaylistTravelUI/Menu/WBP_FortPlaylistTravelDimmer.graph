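# Full-screen backdrop widget for playlist travel, layering a dark Dimmer and a lazy-loaded TravelBackgroundImage, both collapsed at runtime start.
# When the activity view model is set and the destination uses full-screen loading, it fetches images and, absent a TravelingBackgroundImage, shows the Dimmer.
# Once loaded, it swaps the Dimmer for the activity image, passing texture and size to a dynamic material.
# Both graphs take no action when a TravelingBackgroundImage is set.
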WidgetBlueprint /FortPlaylistTravelUI/Menu/WBP_FortPlaylistTravelDimmer
graphs [
 BGImageLoadingState {nodes [
   n1 = BGImageLoadingState!() -> (then, LoadingState:EActivityImageLoadingState Enum=Unset) {is=title K2Node_FunctionEntry}
   > n2 = Switch on EActivityImageLoadingState!(execute, Selection:EActivityImageLoadingState Enum=@n1) -> (Unset:Exec, Loading:Exec, Loaded:Exec, Error:Exec) {is=title K2Node_SwitchEnum}
   n3 = Get BP_FortPlaylistTravelViewModel() -> (BP_FortPlaylistTravelViewModel:BP Fort Playlist Travel View Model Object Reference)
   Branch(
    execute=@n2.Loaded
    Condition=Get bFullScreenLoading!(self:Fort Playlist Travel Entry View Model Object Reference=Get CurrentGameTravelDestinationViewModel!(self:Fort Playlist Travel View Model Object Reference=@n3) -> (CurrentGameTravelDestinationViewModel:Fort Playlist Travel Entry View Model Object Reference)) -> (bFullScreenLoading:bool)
   )
   n4 = Get CurrentGameTravelDestinationViewModel(self=Get BP_FortPlaylistTravelViewModel())
   > n5 = Branch(Condition=IsValidSoftObjectReference(SoftObjectReference=Get TravelingBackgroundImage!(self:Fort Playlist Travel Entry View Model Object Reference=@n4) -> (TravelingBackgroundImage:Soft Object Reference)))
   n6 = GetDynamicMaterial(execute=@n5.else, self=Get TravelBackgroundImage() -> (TravelBackgroundImage:Fort Lazy Image Object Reference))
   n7 = Get ActivityViewModel!(self:Fort Playlist Travel Entry View Model Object Reference=Get CurrentGameTravelDestinationViewModel(self=Get BP_FortPlaylistTravelViewModel())) -> (ActivityViewModel:Discover Activity View Model Object Reference)
   > SetTextureParameterValue(
    self=@n6
    ParameterName=Image
    Value=GetImageTexture!(self:Fort Discover Link Entry Tile Item VM Object Reference=@n7) -> (ReturnValue:Texture Object Reference)
   )
   n8 = BreakVector2D(InVec=GetTextureSize!(self:Fort Discover Link Entry Tile Item VM Object Reference=@n7) -> (ReturnValue:Vector 2D Structure))
   > SetScalarParameterValue(self=@n6, ParameterName=X_TexSize, Value=@n8)
   > SetScalarParameterValue(self=@n6, ParameterName=Y_TexSize, Value=@n8.Y)
   > SetVisibility(self=Get TravelBackgroundImage(), InVisibility=Visible)
   > SetVisibility(self=Get Dimmer() -> (Dimmer:Image Object Reference), InVisibility=Collapsed)
  ]}
 EventGraph {nodes [
   n1 = Event PreConstruct()
   > n2 = Branch(Condition=@n1.IsDesignTime)
   SetVisibility(execute=@n2.else, self=Get TravelBackgroundImage(), InVisibility=Collapsed)
   > SetVisibility(self=Get Dimmer(), InVisibility=Collapsed)
   Event Construct()
   Event Tick()
  ]}
 OnActivityViewModelSet {nodes [
   n1 = OnActivityViewModelSet!() -> (then, ActivityVM:Discover Activity View Model Object Reference) {is=title K2Node_FunctionEntry}
   > Is Valid(InputObject=@n1)
   n2 = Get BP_FortPlaylistTravelViewModel()
   > n3 = Branch(Condition=Get bFullScreenLoading(self=Get CurrentGameTravelDestinationViewModel(self=@n2)))
   > FetchImages(self:Fort Discover Link Entry Tile Item VM Object Reference=Get ActivityVM() -> (ActivityVM:Discover Activity View Model Object Reference))
   n4 = Get CurrentGameTravelDestinationViewModel(self=Get BP_FortPlaylistTravelViewModel())
   > n5 = Branch(Condition=IsValidSoftObjectReference(SoftObjectReference=Get TravelingBackgroundImage(self=@n4)))
   SetVisibility(execute=@n5.else, self=Get TravelBackgroundImage(), InVisibility=Collapsed)
   > SetVisibility(self=Get Dimmer(), InVisibility=Visible)
   SetVisibility(execute=@n3.else, self=Get TravelBackgroundImage(), InVisibility=Collapsed)
   > SetVisibility(self=Get Dimmer(), InVisibility=Collapsed)
  ]}
]
variables []
widgets {
 info {inheritedWidgetCount=0, parentClass=&UserWidget, rootWidgetClass=&Overlay, widgetCount=3}
 widgets [
  Overlay_22 {widgetClassPath=&Overlay, widget=Overlay_22}
  Dimmer {
   widgetClassPath=&Image
   bIsVariable=true
   parent=Overlay_22
   widget=Dimmer
   properties {brush {
     bIsDynamicallyLoaded=false
     drawAs=Image
     imageType=NoImage
     mirroring=NoMirror
     resourceName=None
     resourceObject=null
     tiling=NoTile
     imageSize=<32 32>
     margin=<ltrb 0 0 0 0>
     outlineSettings {
      bUseBrushTransparency=false
      roundingType=HalfHeightRadius
      width=0
      color {colorUseRule=UseColor_Specified, specifiedColor=<rgba 0 0 0 0>}
      cornerRadii=<0 0 0 0>
     }
     tintColor {colorUseRule=UseColor_Specified, specifiedColor=<rgba 0.005 0.005 0.005 1>}
     uVRegion {bIsValid=false, max=<0 0>, min=<0 0>}
    }}
   slot {class=OverlaySlot, properties {horizontalAlignment=HAlign_Fill, verticalAlignment=VAlign_Fill}}
  }
  TravelBackgroundImage {
   widgetClassPath=&FortLazyImage
   bIsVariable=true
   parent=Overlay_22
   widget=TravelBackgroundImage
   properties {
    materialTextureParamName=Image
    brush {
     bIsDynamicallyLoaded=false
     drawAs=Image
     imageType=FullColor
     mirroring=NoMirror
     resourceName=None
     resourceObject=/FortPlaylistTravelUI/Art/MI_UI_FortPlaylistTravel_GamePreviewBackgroundImage
     tiling=NoTile
     imageSize=<32 32>
     margin=<ltrb 0 0 0 0>
     outlineSettings {
      bUseBrushTransparency=false
      roundingType=HalfHeightRadius
      width=0
      color {colorUseRule=UseColor_Specified, specifiedColor=<rgba 0 0 0 0>}
      cornerRadii=<0 0 0 0>
     }
     tintColor {colorUseRule=UseColor_Specified, specifiedColor=<rgba 1 1 1 1>}
     uVRegion {bIsValid=false, max=<0 0>, min=<0 0>}
    }
   }
   slot {class=OverlaySlot, properties {horizontalAlignment=HAlign_Fill, verticalAlignment=VAlign_Fill}}
  }
 ]
}
