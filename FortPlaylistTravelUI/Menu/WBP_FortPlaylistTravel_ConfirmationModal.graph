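# Playlist travel confirmation modal widget, built on FortPlaylistTravelUISelectionModal, showing a rich-text description and a row of selection buttons over a backplate.
# At runtime PreConstruct clears placeholder buttons, and Construct plays an intro that slides and fades the panel before fading in the text.
# Parent events set the description, leaving the title unused, and add each CTA button to the row, using a counter to left-align and fill the first.
# A visibility border limits the inherited close button to touch hardware.
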WidgetBlueprint /FortPlaylistTravelUI/Menu/WBP_FortPlaylistTravel_ConfirmationModal
animations [Anim_Intro {
  binding_count=2
  length_seconds=0.42
  object=&self:Anim_Intro
  sequence {
   bindings [
    RichText_Description {
     id=#1
     possessed_class=RichTextBlock
     tracks [{
       type=MovieSceneFloatTrack
       display_name=Render Opacity
       sections [{type=MovieSceneFloatSection, channels [None {type=MovieSceneScriptingFloatChannel, keys [[24.9983,0],[39.9983,1]]}]}]
      }]
    }
    SizeBox_0 {
     id=#2
     possessed_class=SizeBox
     tracks [
      {
       type=MovieScene2DTransformTrack
       sections [{
         type=MovieScene2DTransformSection
         channels [
          Translation.X {type=MovieSceneScriptingFloatChannel}
          Translation.Y {type=MovieSceneScriptingFloatChannel, keys [[9.9983,160],[24.9983,0]]}
          Scale.X {type=MovieSceneScriptingFloatChannel}
          Scale.Y {type=MovieSceneScriptingFloatChannel}
          Shear.X {type=MovieSceneScriptingFloatChannel}
          Shear.Y {type=MovieSceneScriptingFloatChannel}
         ]
        }]
      }
      {
       type=MovieSceneFloatTrack
       display_name=Render Opacity
       sections [{type=MovieSceneFloatSection, channels [None {type=MovieSceneScriptingFloatChannel, keys [[0,0],[9.9983,0],[24.9983,1]]}]}]
      }
     ]
    }
   ]
   display_rate {numerator=100}
  }
 }]
graphs [EventGraph {nodes [
   n1 = Event PreConstruct()
   > Branch(Condition=NOT Boolean(A=@n1.IsDesignTime))
   > ClearChildren(self=Get ButtonContainer() -> (ButtonContainer:Horizontal Box Object Reference))
   Event Construct()
   > PlayAnimation(InAnimation=Get Anim_Intro() -> (Anim_Intro:Widget Animation Object Reference))
   n2 = Event OnSetTitleAndDescription!() -> (OutputDelegate, then, Title:"Text (by ref)", Description:"Text (by ref)") {is=title K2Node_Event}
   > SetText(self=Get RichText_Description() -> (RichText_Description:Rich Text Block Object Reference), InText=@n2.Description)
   n3 = Event OnSelectionButtonCreated!() -> (OutputDelegate, then, NewButton:Common Button Base Object Reference, SelectionTitle:"Text (by ref)") {is=title K2Node_Event}
   > n4 = Cast To WBP_UIKit_ButtonCTA_Base!(execute, Object=@n3.NewButton) -> (then, CastFailed, AsWBP UIKit Button CTA Base:WBP UIKit Button CTA Base Object Reference) {is=title K2Node_DynamicCast}
   > n5 = AddChildToHorizontalBox!(execute, self:Horizontal Box Object Reference=Get ButtonContainer(), Content:Widget Object Reference=@n4) -> (then, ReturnValue:Horizontal Box Slot Object Reference)
   > SetPadding!(
    execute
    self:Horizontal Box Slot Object Reference=@n5
    InPadding_Left:float=0
    InPadding_Top:float=0
    InPadding_Right:float=0
    InPadding_Bottom:float=0
   ) -> (then)
   > SetText!(execute, self:Fort CTAButton Object Reference=@n4, ButtonText:"Text (by ref)"=@n3.SelectionTitle) -> (then)
   > n6 = Sequence()
   > Branch(Condition="integer <= integer"(A=Get CurrButtonCount() -> (CurrButtonCount:int=0)))
   > SetHorizontalAlignment(self:Horizontal Box Slot Object Reference=@n5, InHorizontalAlignment=HAlign_Left)
   > SetSize!(
    execute
    self:Horizontal Box Slot Object Reference=@n5
    InSize:Slate Child Size Structure=Make Slate Child Size!(Value:float=1, SizeRule:ESlateSizeRule Enum=Fill) -> (SlateChildSize:Slate Child Size Structure) {is=title K2Node_MakeStruct}
   ) -> (then)
   Increment Int(" "=@n6.then_1, Value=Get CurrButtonCount() -> (CurrButtonCount=0))
   n7 = Get CurrButtonCount() -> (CurrButtonCount=0)
   BreakVector(InVec=SelectVector(
     A=Get ButtonPadding_LeftRight_Later() -> (ButtonPadding_LeftRight_Later:Vector)
     B=Get ButtonPadding_LeftRight_First() -> (ButtonPadding_LeftRight_First:Vector)
     bPickA="integer > integer"(A=@n7)
    ))
  ]}]
variables [
 ButtonPadding_LeftRight_First {type=Vector}
 ButtonPadding_LeftRight_Later {type=Vector}
 CurrButtonCount {type=integer}
]
widgets {
 info {
  inheritedWidgetCount=1
  parentClass=/Script/FortPlaylistTravelUI.FortPlaylistTravelUISelectionModal
  rootWidgetClass=&Overlay
  widgetCount=13
 }
 widgets [
  Overlay_0 {widgetClassPath=&Overlay, widget=Overlay_0}
  CHVB_OnlyTouch {
   widgetClassPath=&CommonHardwareVisibilityBorder
   parent=Overlay_0
   widget=CHVB_OnlyTouch
   properties {visibilityQuery {autoDescription=" ANY( )", tokenStreamVersion=0, userDescription=touch-only, queryTokenStream [0,1,1,0], tagDictionary []}}
   slot {class=OverlaySlot, properties {horizontalAlignment=HAlign_Fill, verticalAlignment=VAlign_Fill}}
  }
  Button_TouchClose {
   widgetClassPath=&/WBP_PlaylistTravelUI_CloseButton
   bInherited=true
   bIsVariable=true
   parent=CHVB_OnlyTouch
   widget=Button_TouchClose
   slot {class=BorderSlot, properties {padding=<ltrb 0 0 0 0>}}
  }
  SizeBox_0 {
   widgetClassPath=&SizeBox
   parent=Overlay_0
   widget=SizeBox_0
   properties {
    bOverride_MaxAspectRatio=true
    bOverride_MaxDesiredHeight=true
    bOverride_MinAspectRatio=true
    bOverride_WidthOverride=true
    maxAspectRatio=2.75
    maxDesiredHeight=746
    minAspectRatio=2.75
    widthOverride=1424
   }
   slot {class=OverlaySlot, properties {horizontalAlignment=HAlign_Center, verticalAlignment=VAlign_Center}}
  }
  Overlay_1 {widgetClassPath=&Overlay, parent=SizeBox_0, widget=Overlay_1, slot {class=SizeBoxSlot}}
  WBP_UIKit_Backplate {
   widgetClassPath=&WBP_UIKit_Backplate
   bIsVariable=true
   parent=Overlay_1
   widget=WBP_UIKit_Backplate
   properties {
    brightness=Obscure
    cornerRadius=Interactive Component
    colorAndOpacity=<rgba 1 1 1 0>
    foregroundColor {colorUseRule=UseColor_Foreground, specifiedColor=<rgba 1 0 1 0>}
   }
   slot {class=OverlaySlot, properties {horizontalAlignment=HAlign_Fill, verticalAlignment=VAlign_Fill}}
  }
  HorizontalBox_51 {
   widgetClassPath=&HorizontalBox
   parent=Overlay_1
   widget=HorizontalBox_51
   slot {class=OverlaySlot, properties {horizontalAlignment=HAlign_Fill, verticalAlignment=VAlign_Fill}}
  }
  VB_MainBox {
   widgetClassPath=&VerticalBox
   parent=HorizontalBox_51
   widget=VB_MainBox
   slot {class=HorizontalBoxSlot, properties {padding=<ltrb 48 150 48 40>, size {sizeRule=Fill, value=1}}}
  }
  RichText_Description {
   widgetClassPath=/Script/UMG.RichTextBlock
   bIsVariable=true
   parent=VB_MainBox
   widget=RichText_Description
   properties {
    autoWrapText=true
    bOverrideDefaultStyle=true
    clipping=ClipToBounds
    lineHeightPercentage=2
    text="[PH] This is <emphasis>some</> text that would be important for this modal! "
    textStyleSet=/FortPlaylistTravelUI/DataTables/DT_FortPlaylistTravel_RichTextStyleSet
    defaultTextStyleOverride {
     overflowPolicy=Clip
     transformPolicy=None
     colorAndOpacity {colorUseRule=UseColor_Specified, specifiedColor=<rgba 0.5776 0.5776 0.5776 1>}
     font {
      bForceMonospaced=false
      bMaterialIsStencil=false
      fontMaterial=null
      fontObject=&HeadingNow
      letterSpacing=0
      monospacedWidth=1
      size=20.25
      skewAmount=0
      typefaceFontName=74 Regular
      outlineSettings {
       bApplyOutlineToDropShadows=false
       bMiteredCorners=false
       bSeparateFillAlpha=false
       outlineMaterial=null
       outlineSize=0
       outlineColor=<rgba 0 0 0 1>
      }
     }
     highlightColor {colorUseRule=UseColor_Specified, specifiedColor=<rgba 0 0 0 1>}
     highlightShape {
      bIsDynamicallyLoaded=false
      drawAs=Image
      imageType=NoImage
      mirroring=NoMirror
      resourceName=None
      resourceObject=null
      tiling=NoTile
      imageSize=<32 32>
      margin=<ltrb 0 0 0 0>
      outlineSettings {
       bUseBrushTransparency=false
       roundingType=HalfHeightRadius
       width=0
       color {colorUseRule=UseColor_Specified, specifiedColor=<rgba 0 0 0 0>}
       cornerRadii=<0 0 0 0>
      }
      tintColor {colorUseRule=UseColor_Specified, specifiedColor=<rgba 1 1 1 0>}
      uVRegion {bIsValid=false, max=<0 0>, min=<0 0>}
     }
     selectedBackgroundColor {colorUseRule=UseColor_Foreground, specifiedColor=<rgba 1 0 1 1>}
     shadowColorAndOpacity=<rgba 0 0 0 1>
     shadowOffset=<0 0>
     strikeBrush {
      bIsDynamicallyLoaded=false
      drawAs=Image
      imageType=NoImage
      mirroring=NoMirror
      resourceName=None
      resourceObject=null
      tiling=NoTile
      imageSize=<32 32>
      margin=<ltrb 0 0 0 0>
      outlineSettings {
       bUseBrushTransparency=false
       roundingType=HalfHeightRadius
       width=0
       color {colorUseRule=UseColor_Specified, specifiedColor=<rgba 0 0 0 0>}
       cornerRadii=<0 0 0 0>
      }
      tintColor {colorUseRule=UseColor_Specified, specifiedColor=<rgba 1 1 1 0>}
      uVRegion {bIsValid=false, max=<0 0>, min=<0 0>}
     }
     underlineBrush {
      bIsDynamicallyLoaded=false
      drawAs=Image
      imageType=NoImage
      mirroring=NoMirror
      resourceName=None
      resourceObject=null
      tiling=NoTile
      imageSize=<32 32>
      margin=<ltrb 0 0 0 0>
      outlineSettings {
       bUseBrushTransparency=false
       roundingType=HalfHeightRadius
       width=0
       color {colorUseRule=UseColor_Specified, specifiedColor=<rgba 0 0 0 0>}
       cornerRadii=<0 0 0 0>
      }
      tintColor {colorUseRule=UseColor_Specified, specifiedColor=<rgba 1 1 1 1>}
      uVRegion {bIsValid=false, max=<0 0>, min=<0 0>}
     }
    }
   }
   slot {class=VerticalBoxSlot, properties {size {sizeRule=Fill, value=1}}}
  }
  ButtonContainer {widgetClassPath=&HorizontalBox, bIsVariable=true, parent=VB_MainBox, widget=ButtonContainer, slot {class=VerticalBoxSlot}}
  Button0 {
   widgetClassPath=&WBP_UIKit_Button_Regular
   parent=ButtonContainer
   widget=Button0
   properties {buttonSize=Small, mobileMinHeight=60, overrideDefaultHeight=true, text=Regular}
   slot {class=HorizontalBoxSlot}
  }
  Button1 {
   widgetClassPath=&WBP_UIKit_Button_Regular
   parent=ButtonContainer
   widget=Button1
   properties {buttonSize=Small, buttonType=Quiet, mobileMinHeight=60, overrideDefaultHeight=true, text=Regular}
   slot {class=HorizontalBoxSlot}
  }
  Image_GameImage {
   widgetClassPath=&FortLazyImage
   bIsVariable=true
   parent=HorizontalBox_51
   widget=Image_GameImage
   properties {
    bOverride_Cursor=true
    materialTextureParamName=Image
    visibility=Hidden
    brush {
     bIsDynamicallyLoaded=false
     drawAs=Box
     imageType=FullColor
     mirroring=NoMirror
     resourceName=None
     resourceObject=null
     tiling=NoTile
     imageSize=<512 512>
     margin=<ltrb 0.5 0.5 0.5 0.5>
     outlineSettings {
      bUseBrushTransparency=false
      roundingType=HalfHeightRadius
      width=0
      color {colorUseRule=UseColor_Specified, specifiedColor=<rgba 0 0 0 0>}
      cornerRadii=<0 0 0 0>
     }
     tintColor {colorUseRule=UseColor_Specified, specifiedColor=<rgba 1 1 1 1>}
     uVRegion {bIsValid=false, max=<0 0>, min=<0 0>}
    }
    colorAndOpacity=<rgba 1 1 1 0>
   }
   slot {class=HorizontalBoxSlot, properties {size {sizeRule=Automatic, value=0.5}}}
  }
 ]
}
