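# Translucent UI material for the playlist travel UI that draws an animated, horizontally scrolling travel icon clipped to a rounded box.
# Looping time, sine-eased and raised to a power, drives an X panner, turning steady time into pause-and-sweep motion with a speed and on/off parameter.
# Wrapped UVs get a gap and center scale, then sample a distance-field texture whose anti-aliased shape is multiplied by an SDF rounded-box mask.
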Material /FortPlaylistTravelUI/Art/M_UI_FortPlaylistTravel_Travel
graph {
 properties {
  bAutomaticallySetUsageInEditor=false
  blendMode=BLEND_Translucent
  materialDomain=MD_UI
  pixelDepthOffsetMode=PDOM_Legacy
  usedShadingModels=Default Lit
 }
 nodes [
  n1 = Multiply(
   A=Time()
   B=Multiply(
    A=ScalarParameter() {defaultValue=-0.375, group=Animation, parameterName=PanningSpeed}
    B=ScalarParameter() {defaultValue=1, group=Animation, parameterName=AnimationActive, sliderMax=1}
   )
  ) {constB=0.25}
  n2 = Power(
   Base=MaterialFunctionCall("Value (0-1)"=Frac(None=@n1)) {materialFunction=&MF_UI_EaseCurves}.EaseInOutSine
   Exp=ScalarParameter() {defaultValue=32, group=Animation, parameterName=AnimPower}
  ) {constExponent=16}
  n3 = Frac(None=MaterialFunctionCall(UV=Panner(Time=@n2) {speedX=1}) {materialFunction=&MF_UI_UVScaling}.Scaled to Fit)
  n4 = ScalarParameter() {defaultValue=1.25, parameterName=gap increaser}
  n5 = Constant() {r=0.5}
  n6 = Multiply(A=@n4, B=@n5)
  n7 = MaterialFunctionCall(
   UVs=Subtract(A=Multiply(A=@n3, B=@n4), B=Subtract(A=@n6, B=@n5))
   Texture Scale=ScalarParameter() {defaultValue=1, parameterName=OverallScale, sliderMax=1}
  ) {materialFunction=&ScaleUVsByCenter}
  Multiply(
   A=MaterialFunctionCall(
    DistanceField=TextureSample(UVs=@n7.UVs) {texture=/FortPlaylistTravelUI/Art/Textures/T_UI_FortPlaylistTravel_Travel}.G
    Threshold=Constant() {r=0.26}
    Smoothing=Constant() {r=0.22}
   ) {materialFunction=&MF_UI_AntiAliasing}.Inner
   B=NamedRerouteUsage()
  )
  n8 = MaterialFunctionCall() {materialFunction=&GetUserInterfaceUV}
  n9 = Multiply(A=Constant() {r=0.95}, B=ScalarParameter() {defaultValue=1, parameterName=OverallScale})
  n10 = Multiply(
   A=ScalarParameter() {defaultValue=0.5, group=Icon Settings, parameterName=CornerRoundness}
   B=ScalarParameter() {defaultValue=1, parameterName=OverallScale}
  )
  NamedRerouteDeclaration(None=MaterialFunctionCall(DistanceField=MaterialFunctionCall(UVs=@n8.Normalized UV, Size=@n9, CornerRoundness=@n10) {materialFunction=&MF_UI_Box}.SDF) {materialFunction=&MF_UI_AntiAliasing}.Inner) {name=RoundedBox_1, nodeColor=<rgba 0 0.5547 1 1>}
  VectorParameter() {parameterName=BackingRGB, defaultValue=<rgba 1 1 1 1>}
 ]
}
