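# Translucent UI material for the FortPlaylistTravel UI that draws a circle or rounded-box shape with a tintable fill and optional outline.
# MF_UI_Box builds distance fields from size, per-corner roundness and thickness parameters, which are anti-aliased into fill and outline masks.
# The HasOutline scalar blends the outline into color and alpha.
# The UsesBlur static switch instead blends a Slate post buffer with the fill color, masked by the shape alone.
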Material /FortPlaylistTravelUI/Art/M_UI_FortPlaylistTravel_Circle
graph {
 properties {bAutomaticallySetUsageInEditor=false, blendMode=BLEND_Translucent, materialDomain=MD_UI, usedShadingModels=Default Lit}
 nodes [
  n1 = VectorParameter() {group=Color, parameterName=RBGA_Fill, defaultValue=<rgba 1 1 1 1>}
  n2 = VectorParameter() {group=Color, parameterName=RBGA_Outline, defaultValue=<rgba 1 0.3498 0.7646 1>}
  n3 = MaterialFunctionCall(
   Size=AppendVector(
    A=ScalarParameter() {defaultValue=1, group=Shape, parameterName=Circle_SizeX}
    B=ScalarParameter() {defaultValue=1, group=Shape, parameterName=Circle_SizeY}
   )
   CornerRoundness=MaterialFunctionCall(
    X=ScalarParameter() {defaultValue=1, group=Shape, parameterName=LeftTopCorner}
    Y=ScalarParameter() {defaultValue=1, group=Shape, parameterName=LeftBottomCorner}
    Z=ScalarParameter() {defaultValue=1, group=Shape, parameterName=RightTopCorner}
    A=ScalarParameter() {defaultValue=1, group=Shape, parameterName=RightBottomCorner}
   ) {materialFunction=&MakeFloat4}.Result
   OutlineThickness=ScalarParameter() {defaultValue=0.05, group=Shape, parameterName=OutlineThickness}
   OutlineAlignment=Constant() {r=-1}
   KeepAspectRatio=StaticBoolParameter() {parameterName=KeepAspectRatio}
  ) {materialFunction=&MF_UI_Box}
  n4 = MaterialFunctionCall(DistanceField=@n3.SDFOutline) {materialFunction=&MF_UI_AntiAliasing}
  n5 = Multiply(A=@n2.A, B=@n4.Inner)
  n6 = ScalarParameter() {group=Shape, parameterName=HasOutline}
  StaticSwitchParameter(
   True=LinearInterpolate(
    A=MaterialFunctionCall!("UVs (Optional)") {materialFunction=/Engine/Functions/UserInterface/GetSlatePost0}.LinearRGB
    B=@n1.RGB
    Alpha=@n1.RGB
   )
   False=LinearInterpolate(A=@n1.RGB, B=LinearInterpolate(A=@n1.RGB, B=@n2.RGB, Alpha=@n5), Alpha=@n6)
  ) {parameterName=UsesBlur}
  n7 = MaterialFunctionCall(DistanceField=@n3.SDF) {materialFunction=&MF_UI_AntiAliasing}
  n8 = Multiply(A=@n1.A, B=@n7.Inner)
  n9 = Saturate(None=Subtract(A=@n8, B=@n4.Inner))
  StaticSwitchParameter(True=@n7.Inner, False=LinearInterpolate(A=@n8, B=Add(A=@n9, B=@n5), Alpha=@n6)) {parameterName=UsesBlur}
 ]
}
