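# M_SocialPartyStatusIcon is a translucent UI material, named as the party status icon for the social HUD.
# It contains a sample of the sidebar PartyUp icon texture and a pale blue-grey constant color.
# No links between these two nodes or to the material outputs appear, so how they combine is not shown.
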Material /Game/Athena/HUD/Art/M_SocialPartyStatusIcon
graph {
 properties {blendMode=BLEND_Translucent, materialDomain=MD_UI, pixelDepthOffsetMode=PDOM_Legacy, usedShadingModels=Default Lit}
 nodes [
  Constant3Vector() {constant=<rgba 0.7412 0.7725 0.8588 0>}
  TextureSample() {texture=/Game/UI/Sidebar/Textures/T_UI_Icon_Sidebar_PartyUp}
 ]
}
