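# Player count badge for the Fortnite playlist travel menu, showing a players icon beside "current/max" text in a rounded pill border.
# SetPlayerCountText formats the two counts into the text, then calls ShowErrorState with whether current exceeds max.
# ShowErrorState switches the text, icon and background between red for over capacity and green otherwise.
# The event graph holds no logic, so updates come only through these functions.
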WidgetBlueprint /FortPlaylistTravelUI/Menu/WBP_FortPlaylistTravel_PlayerCountWidget
graphs [
 EventGraph {nodes [
   Event PreConstruct()
   Event Construct()
   Event Tick()
  ]}
 SetPlayerCountText {nodes [
   n1 = SetPlayerCountText!() -> (then, CurrCount:int, MaxCount:int) {is=title K2Node_FunctionEntry}
   > "SetText (Text)"(
    self=Get Text_PlayerCount() -> (Text_PlayerCount:Common Text Object Reference)
    InText=Format Text!(Format="LOCTEXT(\"{Current}/{Max}\")", Current:int=@n1, Max:int=@n1.MaxCount) -> (Result)
   )
   > ShowErrorState!(execute, self, bShow:bool="integer > integer"(A=Get CurrCount() -> (CurrCount:int=0), B=Get MaxCount() -> (MaxCount:int=0))) -> (then)
  ]}
 ShowErrorState {nodes [
   n1 = ShowErrorState!() -> (then, bShow:bool=~) {is=title K2Node_FunctionEntry}
   > SetColorAndOpacity(
    self=Get Text_PlayerCount()
    InColorAndOpacity=Make Slate Color(SpecifiedColor=SelectColor(A=(R=0.1046,G=0,B=0.0009,A=1), B=(R=0.0006,G=0.0782,B=0.0009,A=1), bPickA=@n1))
   )
   > SetBrushColor(
    self=Get Border_BG() -> (Border_BG:Border Object Reference)
    InBrushColor=SelectColor(A=(R=1,G=0.147,B=0.1329,A=1), B=(R=0.0395,G=1,B=0.2307,A=1), bPickA=Get bShow() -> (bShow:bool))
   )
   > SetColorAndOpacity(
    self:Image Object Reference=Get Image_Players() -> (Image_Players:Fort Mobile Image Object Reference)
    InColorAndOpacity:Linear Color Structure=SelectColor(A=(R=0.1046,G=0,B=0.0009,A=1), B=(R=0.0006,G=0.0782,B=0.0009,A=1), bPickA=Get bShow())
   )
  ]}
]
variables []
widgets {
 info {inheritedWidgetCount=0, parentClass=&CommonUserWidget, rootWidgetClass=&Border, widgetCount=4}
 widgets [
  Border_BG {
   widgetClassPath=&Border
   bIsVariable=true
   widget=Border_BG
   properties {
    background {
     bIsDynamicallyLoaded=false
     drawAs=RoundedBox
     imageType=NoImage
     mirroring=NoMirror
     resourceName=None
     resourceObject=null
     tiling=NoTile
     imageSize=<32 32>
     margin=<ltrb 0 0 0 0>
     outlineSettings {
      bUseBrushTransparency=false
      roundingType=HalfHeightRadius
      width=0
      color {colorUseRule=UseColor_Specified, specifiedColor=<rgba 0 0 0 0>}
      cornerRadii=<0 0 0 0>
     }
     tintColor {colorUseRule=UseColor_Specified, specifiedColor=<rgba 1 1 1 1>}
     uVRegion {bIsValid=false, max=<0 0>, min=<0 0>}
    }
    brushColor=<rgba 0.0395 1 0.2307 1>
    padding=<ltrb 10 4 10 4>
   }
  }
  HorizontalBox_41 {
   widgetClassPath=&HorizontalBox
   parent=Border_BG
   widget=HorizontalBox_41
   slot {class=BorderSlot, properties {padding=<ltrb 10 4 10 4>}}
  }
  Image_Players {
   widgetClassPath=/Script/FortniteUI.FortMobileImage
   bIsVariable=true
   parent=HorizontalBox_41
   widget=Image_Players
   properties {
    mobileSizeMultiplier=1.3
    visibility=SelfHitTestInvisible
    brush {
     bIsDynamicallyLoaded=false
     drawAs=Image
     imageType=FullColor
     mirroring=NoMirror
     resourceName=None
     resourceObject=/Game/Athena/HUD/Art/M_SocialPartyStatusIcon
     tiling=NoTile
     imageSize=<32 32>
     margin=<ltrb 0 0 0 0>
     outlineSettings {
      bUseBrushTransparency=false
      roundingType=HalfHeightRadius
      width=0
      color {colorUseRule=UseColor_Specified, specifiedColor=<rgba 0 0 0 0>}
      cornerRadii=<0 0 0 0>
     }
     tintColor {colorUseRule=UseColor_Specified, specifiedColor=<rgba 1 1 1 1>}
     uVRegion {bIsValid=false, max=<0 0>, min=<0 0>}
    }
    colorAndOpacity=<rgba 0.0006 0.0782 0.0009 1>
   }
   slot {class=HorizontalBoxSlot, properties {padding=<ltrb 0 0 5 0>}}
  }
  Text_PlayerCount {
   widgetClassPath=&CommonTextBlock
   bIsVariable=true
   parent=HorizontalBox_41
   widget=Text_PlayerCount
   properties {
    mobileFontSizeMultiplier=1.5
    text=99/99
    colorAndOpacity {colorUseRule=UseColor_Specified, specifiedColor=<rgba 0.0006 0.0782 0.0009 1>}
    font {
     bForceMonospaced=false
     bMaterialIsStencil=false
     fontMaterial=null
     fontObject=&HeadingNow
     letterSpacing=0
     monospacedWidth=1
     size=15
     skewAmount=0
     typefaceFontName=75 Medium
     outlineSettings {
      bApplyOutlineToDropShadows=false
      bMiteredCorners=false
      bSeparateFillAlpha=false
      outlineMaterial=null
      outlineSize=0
      outlineColor=<rgba 0 0 0 1>
     }
    }
    shadowOffset=<0 0>
   }
   slot {class=HorizontalBoxSlot, properties {verticalAlignment=VAlign_Center}}
  }
 ]
}
