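# UI Kit base tag widget: a sheared background pill with tag text and optional icon, plus a "Label - Style" caption beside or below it.
# PreConstruct applies the exposed variables through setter functions.
# SetTagSize and SetTagType both call RefreshTag, which derives colors, fonts, casing and padding from the size and type enums.
# AddIcon constructs the icon at runtime, left or right of the text, gated by UseIcon and ShouldShowLinkCategoryIcon.
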WidgetBlueprint /FortUILibrary/Systems/Tag/WBP_UIKit_Tag_Base
graphs [
 AddIcon {nodes [
   AddIcon() {is=title K2Node_FunctionEntry}
   > RemoveChild(
    self=Get GridText_AndOptionalIcon() -> (GridText_AndOptionalIcon:Grid Panel Object Reference)
    Content=Get IconWidget() -> (IconWidget:Image Object Reference)
   )
   > Branch(Condition=AND Boolean(A=Get UseIcon() -> (UseIcon:bool), B=ShouldShowLinkCategoryIcon!() -> (ReturnValue:bool)))
   > n1 = Construct Object from Class(Class=&Image) -> (ReturnValue:Image Object Reference)
   > n2 = Set IconWidget(IconWidget:Image Object Reference=@n1) -> (Output_Get:Image Object Reference)
   > Branch(Condition=IsValid(Object=@n2))
   > AddChildToGrid!(
    execute
    self:Grid Panel Object Reference=Get GridText_AndOptionalIcon()
    Content:Widget Object Reference=@n2
    InRow:int=0
    InColumn:int=SelectInt(A=0, B=1, bPickA=Get IconLeftAligned() -> (IconLeftAligned:bool))
   ) -> (then, ReturnValue:Grid Slot Object Reference)
   > SetGridSlotColumnAndRow!(
    execute
    Widget:Widget Object Reference=Get Text_Tag() -> (Text_Tag:Common Text Object Reference)
    Row:int=0
    Row Span:int=0
    Column:int=SelectInt(A=1, B=0, bPickA=Get IconLeftAligned())
    Column Span:int=1
   ) -> (then)
   > Set VerticalAlignment!(execute, VerticalAlignment:EVerticalAlignment Enum=VAlign_Center, self:Grid Slot Object Reference=SlotAsGridSlot(Widget=@n2)) -> (then, Output_Get:EVerticalAlignment Enum=VAlign_Fill)
   > SetIconMaterial!(execute, self, IconMaterial:Material Instance Object Reference=Get IconMaterial() -> (IconMaterial:Material Instance Object Reference)) -> (then)
   > RefreshIconSize()
   > SetIconPadding!(execute, self, Padding:Margin Structure=Get IconPadding() -> (IconPadding:Margin Structure)) -> (then)
  ]}
 EventGraph {nodes [
   Event PreConstruct()
   > InitializeTag()
   > SetTagText!(execute, self, Text:Text=Get TagText() -> (TagText:Text)) -> (then)
   > SetStyleText!(execute, self, Text:Text=Get StyleText() -> (StyleText:Text)) -> (then)
   > SetLabelText!(execute, self, Text:Text=Get LabelText() -> (LabelText:Text)) -> (then)
  ]}
 InitializeTag {nodes [
   InitializeTag() {is=title K2Node_FunctionEntry}
   > n1 = Sequence!(execute) -> (then_0, then_1, then_2:Exec)
   SetTagSize!(execute=@n1.then_1, self, TagSize:E_UI_TagSize Enum=Get TagSize() -> (TagSize:E_UI_TagSize Enum=NewEnumerator0)) -> (then)
   > SetTagType!(execute, self, Type:E_UI_TagType Enum=Get TagType() -> (TagType:E_UI_TagType Enum=NewEnumerator0)) -> (then)
   > SetTagOrientation!(execute, self, Orientation:EOrientation Enum=Get TagOrientation() -> (TagOrientation:EOrientation Enum=Orient_Horizontal)) -> (then)
   SetLabelTextColor!(execute=@n1.then_2, self, Color:Linear Color Structure=Get LabelColor() -> (LabelColor:Linear Color Structure)) -> (then)
   AddIcon(execute=@n1.then_0) {is=title K2Node_CallFunction}
  ]}
 RefreshIconSize {nodes [
   RefreshIconSize() {is=title K2Node_FunctionEntry}
   > Branch(Condition=IsValid(Object=Get IconWidget()))
   > SetDesiredSizeOverride!(
    execute
    self:Image Object Reference=Get IconWidget()
    DesiredSize:Vector 2D Structure=Select!(
     NewEnumerator0:Vector 2D Structure=Get IconSizeCompact() -> (IconSizeCompact:Vector 2D Structure)
     NewEnumerator1:Vector 2D Structure=Get IconSize() -> (IconSize:Vector 2D Structure)
     Index:E_UI_TagSize Enum=Get TagSize() -> (TagSize=NewEnumerator0)
    ) -> (ReturnValue:Vector 2D Structure)
   ) -> (then)
  ]}
 RefreshTag {nodes [
   RefreshTag() {is=title K2Node_FunctionEntry}
   n1 = Select!(
    NewEnumerator0:Linear Color Structure=(R=0.8796,G=0.1274,B=1,A=1)
    NewEnumerator1:Linear Color Structure=(R=0.0048,G=0.004,B=0.0116,A=1)
    NewEnumerator2:Linear Color Structure=(R=1,G=1,B=1,A=0.8)
    NewEnumerator3:Linear Color Structure=(R=1,G=1,B=1,A=0.702)
    Index:E_UI_TagType Enum=Get TagType() -> (TagType=NewEnumerator0)
   ) -> (ReturnValue:Linear Color Structure)
   > SetTagTextColor!(execute, self, Color:Linear Color Structure=@n1) -> (then)
   > SetIconColor!(execute, self, Color:Linear Color Structure=@n1) -> (then)
   n2 = Get Use Dark Background() -> (Use Dark Background:bool)
   > SetTagBackgroundColor!(
    execute
    self
    Color:Linear Color Structure=Select(
     NewEnumerator0=(R=0.0723,G=0.0097,B=0.1981,A=1)
     NewEnumerator1=(R=1,G=1,B=1,A=1)
     NewEnumerator2=Select!(Option 0:Linear Color Structure=(R=1,G=1,B=1,A=0.149), Option 1:Linear Color Structure=(R=0.0048,G=0.004,B=0.0116,A=1), Index:bool=@n2) -> (ReturnValue)
     NewEnumerator3=(R=0,G=0,B=0,A=1)
     Index=Get TagType() -> (TagType=NewEnumerator0)
    )
   ) -> (then)
   > Set RenderOpacity!(
    execute
    RenderOpacity:float=Select!(
     NewEnumerator0:float=1
     NewEnumerator1:float=1
     NewEnumerator2:float=1
     NewEnumerator3:float=0
     Index:E_UI_TagType Enum=Get TagType() -> (TagType=NewEnumerator0)
    ) -> (ReturnValue:float)
    self:Widget Object Reference=Get Background() -> (Background:Image Object Reference)
   ) -> (then, Output_Get:float)
   n3 = Get TagSize() -> (TagSize=NewEnumerator0)
   n4 = Get PrimaryFontCompact() -> (PrimaryFontCompact:Slate Font Info Structure)
   n5 = Get PrimaryFont() -> (PrimaryFont:Slate Font Info Structure)
   n6 = Get SecondaryFontCompact() -> (SecondaryFontCompact:Slate Font Info Structure)
   n7 = Get SecondaryFont() -> (SecondaryFont:Slate Font Info Structure)
   n8 = Get TagSize() -> (TagSize=NewEnumerator0)
   n9 = Select!(
    NewEnumerator0:Slate Font Info Structure=Get RarityFontCompact() -> (RarityFontCompact:Slate Font Info Structure)
    NewEnumerator1:Slate Font Info Structure=Get RarityFont() -> (RarityFont:Slate Font Info Structure)
    Index:E_UI_TagSize Enum=Get TagSize() -> (TagSize=NewEnumerator0)
   ) -> (ReturnValue:Slate Font Info Structure)
   > SetFont(
    self=Get Text_Tag()
    InFontInfo=Select!(
     NewEnumerator0=@n9
     NewEnumerator1=Select!(NewEnumerator0=@n4, NewEnumerator1=@n5, Index:E_UI_TagSize Enum=@n3) -> (ReturnValue)
     NewEnumerator2:Slate Font Info Structure=Select(NewEnumerator0=@n6, NewEnumerator1=@n7, Index=@n8)
     NewEnumerator3:Slate Font Info Structure=@n9
     Index:E_UI_TagType Enum=Get TagType() -> (TagType=NewEnumerator0)
    ) -> (ReturnValue)
   )
   n10 = Get TagSize() -> (TagSize=NewEnumerator0)
   n11 = Get RarityFontCompact()
   n12 = Get RarityFont()
   n13 = Select(NewEnumerator0=Get PrimaryFontCompact(), NewEnumerator1=Get PrimaryFont(), Index=Get TagSize() -> (TagSize=NewEnumerator0))
   > SetFont(
    self=Get Text_Label() -> (Text_Label:Common Text Object Reference)
    InFontInfo=Select!(
     NewEnumerator0=Select!(NewEnumerator0=@n11, NewEnumerator1=@n12, Index:E_UI_TagSize Enum=@n10) -> (ReturnValue)
     NewEnumerator1=@n13
     NewEnumerator2:Slate Font Info Structure=@n13
     NewEnumerator3:Slate Font Info Structure=@n13
     Index:E_UI_TagType Enum=Get TagType() -> (TagType=NewEnumerator0)
    ) -> (ReturnValue)
   )
   > Set TextTransformPolicy!(
    execute
    TextTransformPolicy:ETextTransformPolicy Enum=Select!(
     NewEnumerator0:ETextTransformPolicy Enum=ToUpper
     NewEnumerator1:ETextTransformPolicy Enum=None
     NewEnumerator2:ETextTransformPolicy Enum=None
     NewEnumerator3:ETextTransformPolicy Enum=ToUpper
     Index:E_UI_TagType Enum=Get TagType() -> (TagType=NewEnumerator0)
    ) -> (ReturnValue:ETextTransformPolicy Enum)
    self:Text Object Reference=Get Text_Tag()
   ) -> (then, Output_Get:ETextTransformPolicy Enum=None)
   > Set TextTransformPolicy(
    TextTransformPolicy=Select(NewEnumerator0=ToUpper, NewEnumerator1=None, NewEnumerator2=None, NewEnumerator3=None, Index=Get TagType() -> (TagType=NewEnumerator0))
    self=Get Text_Label()
   ) -> (Output_Get=None)
   > Set Padding!(
    execute
    Padding:Margin Structure=Select!(
     NewEnumerator0:Margin Structure=Get TagCompactPadding() -> (TagCompactPadding:Margin Structure)
     NewEnumerator1:Margin Structure=Get TagPadding() -> (TagPadding:Margin Structure)
     Index:E_UI_TagSize Enum=Get TagSize() -> (TagSize=NewEnumerator0)
    ) -> (ReturnValue:Margin Structure)
    self:Grid Slot Object Reference=SlotAsGridSlot(Widget=Get GridText_AndOptionalIcon())
   ) -> (then, Output_Get:Margin Structure)
   > SetDesiredSizeOverride(
    self=Get Background()
    DesiredSize=Select(NewEnumerator0:Vector 2D Structure=(X=0,Y=24), NewEnumerator1:Vector 2D Structure, Index=Get TagSize() -> (TagSize=NewEnumerator0)) -> (ReturnValue:Vector 2D Structure)
   )
  ]}
 SetIconColor {nodes [
   n1 = SetIconColor!() -> (then, Color:Linear Color Structure) {is=title K2Node_FunctionEntry}
   > n2 = Set IconColor(IconColor:Linear Color Structure=@n1) -> (Output_Get:Linear Color Structure)
   > Branch(Condition=IsValid(Object=Get IconWidget()))
   > Set ColorAndOpacity!(execute, ColorAndOpacity:Linear Color Structure=@n2, self:Image Object Reference=Get IconWidget()) -> (then, Output_Get:Linear Color Structure)
  ]}
 SetIconMaterial {nodes [
   n1 = SetIconMaterial!() -> (then, IconMaterial:Material Instance Object Reference) {is=title K2Node_FunctionEntry}
   > n2 = Set IconMaterial(IconMaterial:Material Instance Object Reference=@n1) -> (Output_Get:Material Instance Object Reference)
   > Branch(Condition=IsValid(Object=Get IconWidget()))
   > SetBrushFromMaterial(self=Get IconWidget(), Material=@n2)
  ]}
 SetIconPadding {nodes [
   n1 = SetIconPadding!() -> (then, Padding:Margin Structure) {is=title K2Node_FunctionEntry}
   > n2 = Set IconPadding(IconPadding:Margin Structure=@n1) -> (Output_Get:Margin Structure)
   > Branch(Condition=IsValid(Object=Get IconWidget()))
   > Set Padding(Padding=@n2, self=SlotAsGridSlot(Widget=Get IconWidget()))
  ]}
 SetLabelText {nodes [
   n1 = SetLabelText!() -> (then, Text:Text) {is=title K2Node_FunctionEntry}
   > Set LabelText(LabelText:Text=@n1) -> (Output_Get:Text)
   n2 = Get StyleText()
   n3 = Get LabelText()
   > Set Text(
    Text=Select!(
     Option 0:Text=Format Text!(Format="LOCTEXT(\"{LabelText} - {StyleText}\")", LabelText:Text=@n3, StyleText:Text=@n2) -> (Result)
     Option 1:Text=@n3
     Index:bool=TextIsEmpty(InText=@n2)
    ) -> (ReturnValue:Text)
    self=Get Text_Label()
   )
  ]}
 SetLabelTextColor {nodes [
   n1 = SetLabelTextColor!() -> (then, Color:Linear Color Structure) {is=title K2Node_FunctionEntry}
   > n2 = Set LabelColor(LabelColor:Linear Color Structure=@n1) -> (Output_Get:Linear Color Structure)
   > Set ColorAndOpacity(ColorAndOpacity:Slate Color Structure=Make Slate Color(SpecifiedColor=@n2), self:Text Object Reference=Get Text_Label()) -> (Output_Get:Slate Color Structure)
  ]}
 SetStyleText {nodes [
   n1 = SetStyleText!() -> (then, Text:Text) {is=title K2Node_FunctionEntry}
   > Set StyleText(StyleText:Text=@n1) -> (Output_Get:Text)
   n2 = Get StyleText()
   n3 = Get LabelText()
   > Set Text(
    Text=Select(
     Option 0=Format Text(Format="LOCTEXT(\"{LabelText} - {StyleText}\")", LabelText=@n3, StyleText=@n2)
     Option 1=@n3
     Index=TextIsEmpty(InText=@n2)
    )
    self=Get Text_Label()
   )
  ]}
 SetStyleTextColor {nodes [
   n1 = SetStyleTextColor!() -> (then, Color:Linear Color Structure) {is=title K2Node_FunctionEntry}
   > n2 = Set StyleColor(StyleColor:Linear Color Structure=@n1) -> (Output_Get:Linear Color Structure)
   > Set ColorAndOpacity(ColorAndOpacity=Make Slate Color(SpecifiedColor=@n2), self=Get Text_Label())
  ]}
 SetTagBackgroundColor {nodes [
   n1 = SetTagBackgroundColor!() -> (then, Color:Linear Color Structure) {is=title K2Node_FunctionEntry}
   > n2 = Set BackgroundColor(BackgroundColor:Linear Color Structure=@n1) -> (Output_Get:Linear Color Structure)
   > Set ColorAndOpacity(ColorAndOpacity:Linear Color Structure=@n2, self:Image Object Reference=Get Background()) -> (Output_Get:Linear Color Structure)
  ]}
 SetTagOrientation {nodes [
   n1 = SetTagOrientation!() -> (then, Orientation:EOrientation Enum=Orient_Horizontal) {is=title K2Node_FunctionEntry}
   > n2 = Set TagOrientation(TagOrientation:EOrientation Enum=@n1) -> (Output_Get:EOrientation Enum=Orient_Horizontal)
   > Set Orientation!(execute, Orientation:EOrientation Enum=@n2, self:Stack Box Object Reference=Get Root() -> (Root:Stack Box Object Reference)) -> (then, Output_Get:EOrientation Enum=Orient_Horizontal)
   > n3 = Cast To StackBoxSlot!(
    execute
    Object=Get Slot!(self:Widget Object Reference=Get TagContainer() -> (TagContainer:Grid Panel Object Reference)) -> (Slot:Panel Slot Object Reference)
   ) -> (then, CastFailed, AsStack Box Slot:Stack Box Slot Object Reference) {is=title K2Node_DynamicCast}
   > Set HorizontalAlignment!(
    execute
    HorizontalAlignment:EHorizontalAlignment Enum=Select!(
     Orient_Horizontal:EHorizontalAlignment Enum=HAlign_Left
     Orient_Vertical:EHorizontalAlignment Enum=HAlign_Left
     Index:EOrientation Enum=Get TagOrientation() -> (TagOrientation=Orient_Horizontal)
    ) -> (ReturnValue:EHorizontalAlignment Enum)
    self:Stack Box Slot Object Reference=@n3
   ) -> (then, Output_Get:EHorizontalAlignment Enum=HAlign_Fill)
   > n4 = Cast To StackBoxSlot(Object=Get Slot(self=Get Text_Label()))
   > Set HorizontalAlignment(
    HorizontalAlignment=Select(Orient_Horizontal=HAlign_Left, Orient_Vertical=HAlign_Left, Index=Get TagOrientation() -> (TagOrientation=Orient_Horizontal))
    self=@n4
   ) -> (Output_Get=HAlign_Fill)
   > Set Padding(
    Padding=Select!(
     Orient_Horizontal_Left:float=8
     Orient_Horizontal_Top:float=0
     Orient_Horizontal_Right:float=0
     Orient_Horizontal_Bottom:float=0
     Orient_Vertical_Left:float=0
     Orient_Vertical_Top:float=4
     Orient_Vertical_Right:float=0
     Orient_Vertical_Bottom:float=0
     Index=Get TagOrientation() -> (TagOrientation=Orient_Horizontal)
    ) -> (ReturnValue:Margin Structure)
    self:Stack Box Slot Object Reference=@n4
   )
  ]}
 SetTagSize {nodes [
   n1 = SetTagSize!() -> (then, TagSize:E_UI_TagSize Enum=NewEnumerator0) {is=title K2Node_FunctionEntry}
   > Set TagSize(TagSize:E_UI_TagSize Enum=@n1) -> (Output_Get:E_UI_TagSize Enum=NewEnumerator0)
   > RefreshTag() {is=title K2Node_CallFunction}
  ]}
 SetTagText {nodes [
   n1 = SetTagText!() -> (then, Text:Text) {is=title K2Node_FunctionEntry}
   > n2 = Set TagText(TagText:Text=@n1) -> (Output_Get:Text)
   > Set Text(Text=@n2, self=Get Text_Tag())
  ]}
 SetTagTextColor {nodes [
   n1 = SetTagTextColor!() -> (then, Color:Linear Color Structure) {is=title K2Node_FunctionEntry}
   > n2 = Set TagColor(TagColor:Linear Color Structure=@n1) -> (Output_Get:Linear Color Structure)
   > Set ColorAndOpacity(ColorAndOpacity:Slate Color Structure=Make Slate Color(SpecifiedColor=@n2), self:Text Object Reference=Get Text_Tag()) -> (Output_Get:Slate Color Structure)
  ]}
 SetTagType {nodes [
   n1 = SetTagType!() -> (then, Type:E_UI_TagType Enum=NewEnumerator0) {is=title K2Node_FunctionEntry}
   > Set TagType(TagType:E_UI_TagType Enum=@n1) -> (Output_Get:E_UI_TagType Enum=NewEnumerator0)
   > RefreshTag()
  ]}
 SetTagVisibility {nodes [
   n1 = SetTagVisibility!() -> (then, Visible:bool=~) {is=title K2Node_FunctionEntry}
   n2 = NOT Boolean(A=TextIsEmpty(InText=Get TagText()))
   > SetVisibility(
    self=Get TagContainer()
    InVisibility="To Visibility (Boolean)"!(
     bIsVisible:bool=AND Boolean(A=@n1, B=@n2)
     TrueVisibility:ESlateVisibility Enum=HitTestInvisible
     FalseVisibility:ESlateVisibility Enum=Collapsed
    ) -> (ReturnValue:ESlateVisibility Enum=Visible)
   )
  ]}
]
variables [
 BackgroundColor {type=LinearColor, default=<rgba 0 0 0 0>}
 BackgroundMaterial {type=MaterialInstance, default=null}
 IconColor {type=LinearColor, default=<rgba 0 0 0 0>}
 IconLeftAligned {type=boolean}
 IconMaterial {type=MaterialInstance, default=/FortUILibrary/Textures/MaterialInstances/MI_UIKit_Icon_LegoIsland}
 IconPadding {type=Margin, default=<ltrb 0 0 0 0>}
 IconSize {type=Vector2D, default=<24 24>}
 IconSizeCompact {type=Vector2D, default=<16 16>}
 IconWidget {type=Image, default=null}
 LabelColor {type=LinearColor, default=<rgba 1 1 1 0.7>}
 LabelPadding {type=Margin, default=<ltrb 8 0 8 0>}
 LabelText {type=string, default="[PH] Label"}
 PrimaryFont {
  type=SlateFontInfo
  default {
   bForceMonospaced=false
   bMaterialIsStencil=false
   fontMaterial=null
   fontObject=&HeadingNow
   letterSpacing=10
   monospacedWidth=1
   size=15.75
   skewAmount=0
   typefaceFontName=75 Medium Italic
   outlineSettings {
    bApplyOutlineToDropShadows=false
    bMiteredCorners=false
    bSeparateFillAlpha=false
    outlineMaterial=null
    outlineSize=0
    outlineColor=<rgba 0 0 0 1>
   }
  }
 }
 PrimaryFontCompact {
  type=SlateFontInfo
  default {
   bForceMonospaced=false
   bMaterialIsStencil=false
   fontMaterial=null
   fontObject=&HeadingNow
   letterSpacing=10
   monospacedWidth=1
   size=12
   skewAmount=0
   typefaceFontName=75 Medium Italic
   outlineSettings {
    bApplyOutlineToDropShadows=false
    bMiteredCorners=false
    bSeparateFillAlpha=false
    outlineMaterial=null
    outlineSize=0
    outlineColor=<rgba 0 0 0 1>
   }
  }
 }
 RarityFont {
  type=SlateFontInfo
  default {
   bForceMonospaced=false
   bMaterialIsStencil=false
   fontMaterial=null
   fontObject=&HeadingNow
   letterSpacing=10
   monospacedWidth=1
   size=15.75
   skewAmount=0
   typefaceFontName=86 Bold Italic
   outlineSettings {
    bApplyOutlineToDropShadows=false
    bMiteredCorners=false
    bSeparateFillAlpha=false
    outlineMaterial=null
    outlineSize=0
    outlineColor=<rgba 0 0 0 1>
   }
  }
 }
 RarityFontCompact {
  type=SlateFontInfo
  default {
   bForceMonospaced=false
   bMaterialIsStencil=false
   fontMaterial=null
   fontObject=&HeadingNow
   letterSpacing=10
   monospacedWidth=1
   size=12
   skewAmount=0
   typefaceFontName=86 Bold Italic
   outlineSettings {
    bApplyOutlineToDropShadows=false
    bMiteredCorners=false
    bSeparateFillAlpha=false
    outlineMaterial=null
    outlineSize=0
    outlineColor=<rgba 0 0 0 1>
   }
  }
 }
 SecondaryFont {
  type=SlateFontInfo
  default {
   bForceMonospaced=false
   bMaterialIsStencil=false
   fontMaterial=null
   fontObject=&HeadingNow
   letterSpacing=10
   monospacedWidth=1
   size=15.75
   skewAmount=0
   typefaceFontName=75 Medium Italic
   outlineSettings {
    bApplyOutlineToDropShadows=false
    bMiteredCorners=false
    bSeparateFillAlpha=false
    outlineMaterial=null
    outlineSize=0
    outlineColor=<rgba 0 0 0 1>
   }
  }
 }
 SecondaryFontCompact {
  type=SlateFontInfo
  default {
   bForceMonospaced=false
   bMaterialIsStencil=false
   fontMaterial=null
   fontObject=&HeadingNow
   letterSpacing=10
   monospacedWidth=1
   size=12.75
   skewAmount=0
   typefaceFontName=75 Medium Italic
   outlineSettings {
    bApplyOutlineToDropShadows=false
    bMiteredCorners=false
    bSeparateFillAlpha=false
    outlineMaterial=null
    outlineSize=0
    outlineColor=<rgba 0 0 0 1>
   }
  }
 }
 StyleColor {type=LinearColor, default=<rgba 1 1 1 0.7>}
 StylePadding {type=Margin, default=<ltrb 8 0 8 0>}
 StyleText {type=string, default=""}
 TagColor {type=LinearColor, default=<rgba 0 0 0 0>}
 TagCompactPadding {type=Margin, default=<ltrb 8 0 8 0>}
 TagOrientation {type=EOrientation, default=Orient_Horizontal}
 TagPadding {type=Margin, default=<ltrb 12 0 12 0>}
 TagSize {type=E_UI_TagSize, default=Default}
 TagText {type=string, default=Tag}
 TagType {type=E_UI_TagType, default=Rarity Tag}
 TertiaryFont_Bold {
  type=SlateFontInfo
  default {
   bForceMonospaced=false
   bMaterialIsStencil=false
   fontMaterial=null
   fontObject=&HeadingNow
   letterSpacing=10
   monospacedWidth=1
   size=15.75
   skewAmount=0
   typefaceFontName=75 Medium Italic
   outlineSettings {
    bApplyOutlineToDropShadows=false
    bMiteredCorners=false
    bSeparateFillAlpha=false
    outlineMaterial=null
    outlineSize=0
    outlineColor=<rgba 0 0 0 1>
   }
  }
 }
 TertiaryFont_Regular {
  type=SlateFontInfo
  default {
   bForceMonospaced=false
   bMaterialIsStencil=false
   fontMaterial=null
   fontObject=&HeadingNow
   letterSpacing=10
   monospacedWidth=1
   size=15.75
   skewAmount=0
   typefaceFontName=74 Regular
   outlineSettings {
    bApplyOutlineToDropShadows=false
    bMiteredCorners=false
    bSeparateFillAlpha=false
    outlineMaterial=null
    outlineSize=0
    outlineColor=<rgba 0 0 0 1>
   }
  }
 }
 Use Dark Background {type=boolean}
 UseIcon {type=boolean}
]
widgets {
 info {inheritedWidgetCount=0, parentClass=&UserWidget, rootWidgetClass=&StackBox, widgetCount=6}
 widgets [
  Root {widgetClassPath=&StackBox, bIsVariable=true, widget=Root}
  TagContainer {
   widgetClassPath=&GridPanel
   bIsVariable=true
   parent=Root
   widget=TagContainer
   properties {columnFill [1], rowFill [1]}
   slot {class=StackBoxSlot, properties {horizontalAlignment=HAlign_Left, verticalAlignment=VAlign_Center}}
  }
  Background {
   widgetClassPath=&Image
   bIsVariable=true
   parent=TagContainer
   widget=Background
   properties {
    visibility=SelfHitTestInvisible
    brush {
     bIsDynamicallyLoaded=false
     drawAs=Image
     imageType=FullColor
     mirroring=NoMirror
     resourceName=None
     resourceObject=/FortUILibrary/Systems/Tag/MaterialInstances/MI_UIKit_TagBaseBackground
     tiling=NoTile
     imageSize=<0 0>
     margin=<ltrb 0 0 0 0>
     outlineSettings {
      bUseBrushTransparency=false
      roundingType=HalfHeightRadius
      width=0
      color {colorUseRule=UseColor_Specified, specifiedColor=<rgba 0 0 0 0>}
      cornerRadii=<0 0 0 0>
     }
     tintColor {colorUseRule=UseColor_Specified, specifiedColor=<rgba 1 1 1 1>}
     uVRegion {bIsValid=false, max=<0 0>, min=<0 0>}
    }
    renderTransform {angle=0, scale=<1 1>, shear=<-10 0>, translation=<0 0>}
   }
   slot {class=GridSlot}
  }
  GridText_AndOptionalIcon {
   widgetClassPath=&GridPanel
   bIsVariable=true
   parent=TagContainer
   widget=GridText_AndOptionalIcon
   properties {columnFill [1,0]}
   slot {class=GridSlot, properties {layer=1, verticalAlignment=VAlign_Center}}
  }
  Text_Tag {
   widgetClassPath=&CommonTextBlock
   bIsVariable=true
   parent=GridText_AndOptionalIcon
   widget=Text_Tag
   properties {
    applyLineHeightToBottomLine=false
    bAutoCollapseWithEmptyText=true
    mobileFontSizeMultiplier=1.2
    pixelSnapping=Disabled
    text=Tag
    colorAndOpacity {colorUseRule=UseColor_Specified, specifiedColor=<rgba 1 0 0.9536 1>}
    font {
     bForceMonospaced=false
     bMaterialIsStencil=false
     fontMaterial=null
     fontObject=&HeadingNow
     letterSpacing=0
     monospacedWidth=1
     size=11.81
     skewAmount=0
     typefaceFontName=86 Bold Italic
     outlineSettings {
      bApplyOutlineToDropShadows=false
      bMiteredCorners=false
      bSeparateFillAlpha=false
      outlineMaterial=null
      outlineSize=0
      outlineColor=<rgba 0 0 0 1>
     }
    }
    shadowOffset=<0 0>
   }
   slot {class=GridSlot, properties {columnSpan=1}}
  }
  Text_Label {
   widgetClassPath=&CommonTextBlock
   bIsVariable=true
   parent=Root
   widget=Text_Label
   properties {
    applyLineHeightToBottomLine=false
    bAutoCollapseWithEmptyText=true
    clipping=OnDemand
    mobileFontSizeMultiplier=1.2
    pixelSnapping=Disabled
    scrollStyle=/FortUILibrary/Styles/Typography/Foundation/TextScrollStyle-NoFade.TextScrollStyle-NoFade_C
    text=Really Long Test String Really Long Test String
    textOverflowPolicy=Ellipsis
    colorAndOpacity {colorUseRule=UseColor_Specified, specifiedColor=<rgba 1 1 1 0.7>}
    font {
     bForceMonospaced=false
     bMaterialIsStencil=false
     fontMaterial=null
     fontObject=&HeadingNow
     letterSpacing=0
     monospacedWidth=1
     size=11.81
     skewAmount=0
     typefaceFontName=86 Bold Italic
     outlineSettings {
      bApplyOutlineToDropShadows=false
      bMiteredCorners=false
      bSeparateFillAlpha=false
      outlineMaterial=null
      outlineSize=0
      outlineColor=<rgba 0 0 0 1>
     }
    }
    shadowOffset=<0 0>
   }
   slot {class=StackBoxSlot, properties {horizontalAlignment=HAlign_Left, verticalAlignment=VAlign_Center, padding=<ltrb 8 0 0 0>}}
  }
 ]
}
