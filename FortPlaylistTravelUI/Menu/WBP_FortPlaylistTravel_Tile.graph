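# CommonUI tile for the Playlist Travel menu, showing an experience's title, description, tags, player count, preview image and travel button.
# Setting the activity view model swaps throbber for content, fetches images and broadcasts loaded; without a preview image, the loaded activity texture feeds a dynamic material.
# Travel status drives button text, interactivity, state text and timer visibility, while lock reasons set error text.
# Button clicks are re-broadcast through a delegate.
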WidgetBlueprint /FortPlaylistTravelUI/Menu/WBP_FortPlaylistTravel_Tile
graphs [
 BGImageLoadingState {nodes [
   n1 = BGImageLoadingState!() -> (then, LoadingState:EActivityImageLoadingState Enum=Unset) {is=title K2Node_FunctionEntry}
   > n2 = Switch on EActivityImageLoadingState!(execute, Selection:EActivityImageLoadingState Enum=@n1) -> (Unset:Exec, Loading:Exec, Loaded:Exec, Error:Exec) {is=title K2Node_SwitchEnum}
   n3 = Get BP_FortPlaylistTravelEntryViewModel() -> (BP_FortPlaylistTravelEntryViewModel:BP Fort Playlist Travel Entry View Model Object Reference)
   n4 = Branch(
    execute=@n2.Loaded
    Condition=IsValidSoftObjectReference(SoftObjectReference=Get PreviewImage!(self:Fort Playlist Travel Entry View Model Object Reference=@n3) -> (PreviewImage:Soft Object Reference))
   )
   n5 = GetDynamicMaterial(execute=@n4.else, self=Get Image_GameImage() -> (Image_GameImage:Fort Lazy Image Object Reference))
   n6 = Get ActivityViewModel!(self:Fort Playlist Travel Entry View Model Object Reference=Get BP_FortPlaylistTravelEntryViewModel()) -> (ActivityViewModel:Discover Activity View Model Object Reference)
   > SetTextureParameterValue(
    self=@n5
    ParameterName=Image
    Value=GetImageTexture!(self:Fort Discover Link Entry Tile Item VM Object Reference=@n6) -> (ReturnValue:Texture Object Reference)
   )
   n7 = BreakVector2D(InVec=GetTextureSize!(self:Fort Discover Link Entry Tile Item VM Object Reference=@n6) -> (ReturnValue:Vector 2D Structure))
   > SetScalarParameterValue(self=@n5, ParameterName=X_TexSize, Value=@n7)
   > SetScalarParameterValue(self=@n5, ParameterName=Y_TexSize, Value=@n7.Y)
   > SetVisibility(self=Get Image_GameImage(), InVisibility=Visible)
  ]}
 BuildActivityTags {nodes [
   BuildActivityTags!() -> (then, TagArray:Array of Creative Island Description Tag Structures) {is=title K2Node_FunctionEntry}
   > Reset!(
    execute
    self:Dynamic Entry Box Object Reference=Get EntryBox_ActivityTags() -> (EntryBox_ActivityTags:Dynamic Entry Box Object Reference)
    bDeleteWidgets:bool=false
   ) -> (then)
   > n1 = For Each Loop(Array:Array of Creative Island Description Tag Structures=Get TagArray() -> (TagArray:Array of Creative Island Description Tag Structures)) -> (Array Element:Creative Island Description Tag Structure)
   > n2 = Create Entry of Class!(
    execute
    self:Dynamic Entry Box Object Reference=Get EntryBox_ActivityTags()
    EntryClass:User Widget Class Reference=/FortUILibrary/Systems/Tag/WBP_UIKit_Tag_Base.WBP_UIKit_Tag_Base_C
   ) -> (then, ReturnValue:WBP UIKit Tag Base Object Reference)
   > SetTagText!(
    execute
    self:WBP UIKit Tag Base Object Reference=@n2
    Text:Text=Break Creative Island Description Tag!(CreativeIslandDescriptionTag:"Creative Island Description Tag Structure (by ref)"=@n1) -> (Name:String, DisplayText:Text) {is=title K2Node_BreakStruct}.DisplayText
   ) -> (then)
   > SetLabelText!(execute, self:WBP UIKit Tag Base Object Reference=@n2, Text:Text) -> (then)
   > SetTagType!(execute, self:WBP UIKit Tag Base Object Reference=@n2, Type:E_UI_TagType Enum=NewEnumerator1) -> (then)
   > SetTagSize!(execute, self:WBP UIKit Tag Base Object Reference=@n2, TagSize:E_UI_TagSize Enum=NewEnumerator0) -> (then)
  ]}
 EventGraph {nodes [
   "On Clicked (Button_Travel)"!() -> (OutputDelegate, then, Button:Common Button Base Object Reference) {is=title K2Node_ComponentBoundEvent}
   > Call On Travel Button Clicked!(execute:Exec, self:WBP Fort Playlist Travel Tile Object Reference) -> (then:Exec) {is=title K2Node_CallDelegate}
   Event Construct()
  ]}
 HandleGameTravelStatus {nodes [
   n1 = HandleGameTravelStatus!() -> (then, Status:EFortPlaylistTravelStatus Enum=None) {is=title K2Node_FunctionEntry}
   > n2 = Sequence!(execute) -> (then_0, then_1, then_2:Exec)
   n3 = Get Status() -> (Status:EFortPlaylistTravelStatus Enum=None)
   > FortShippingLog!(
    execute
    WorldContextObject:Object Reference=Self-Reference()
    InString:String=Append(
     A="[PlaylistTravel] [HUD] [TravelTile] HandleGameTravelStatus: "
     B=Enum to String!(Enumerator:EFortPlaylistTravelStatus Enum=@n3) -> (ReturnValue:String)
    )
    bLogAsWarning:bool=false
   ) -> (then)
   SetText!(
    execute=@n2.then_1
    self:Fort CTAButton Object Reference=Get Button_Travel() -> (Button_Travel:WBP UIKit Button Loud Object Reference)
    ButtonText:"Text (by ref)"=Select!(
     None:Text=Get ButtonText_Travel() -> (ButtonText_Travel:Text)
     PrepTime:Text=Get ButtonText_Cancel() -> (ButtonText_Cancel:Text)
     Launching:Text=Get ButtonText_Launching() -> (ButtonText_Launching:Text)
     Matchmaking:Text
     MatchmakingSuccess:Text
     MatchmakingFailed:Text
     Index:EFortPlaylistTravelStatus Enum=@n1
    ) -> (ReturnValue:Text)
   ) -> (then)
   > n4 = Branch(Condition="Equal (Enum)"(A:EFortPlaylistTravelStatus Enum=Get Status() -> (Status=None), B:EFortPlaylistTravelStatus Enum=None))
   > SetVisibility(self=Get HorizontalBox_TravelState() -> (HorizontalBox_TravelState:Horizontal Box Object Reference), InVisibility=Hidden)
   SetVisibility(execute=@n4.else, self=Get HorizontalBox_TravelState(), InVisibility=HitTestInvisible)
   > "SetText (Text)"(
    self=Get Text_TravelState() -> (Text_TravelState:Common Text Object Reference)
    InText=Select(
     PrepTime=Get PreparingText() -> (PreparingText:Text)
     Launching=Get LaunchingText() -> (LaunchingText:Text)
     Matchmaking=Get MatchmakingText() -> (MatchmakingText:Text)
     Index=Get Status() -> (Status=None)
    )
   )
   n5 = Get Status() -> (Status=None)
   > SetVisibility(
    self=Get Text_Timer() -> (Text_Timer:Common Text Object Reference)
    InVisibility=Select!(Option 0:ESlateVisibility Enum=Collapsed, Option 1:ESlateVisibility Enum=HitTestInvisible, Index:bool="Equal (Enum)"(A=@n5, B=PrepTime)) -> (ReturnValue:ESlateVisibility Enum)
   )
   SetIsInteractionEnabled!(
    execute=@n2.then_2
    self:Common Button Base Object Reference=Get Button_Travel()
    bInIsInteractionEnabled:bool=Select!(
     None:bool=true
     PrepTime:bool=true
     Launching:bool=false
     Matchmaking:bool=false
     MatchmakingSuccess:bool=false
     MatchmakingFailed:bool=true
     Index:EFortPlaylistTravelStatus Enum=Get Status() -> (Status=None)
    ) -> (ReturnValue:bool)
   ) -> (then)
  ]}
 HandleImageFromViewmodel {nodes [
   n1 = HandleImageFromViewmodel!() -> (then, PreviewImage:Soft Object Reference) {is=title K2Node_FunctionEntry}
   > Branch(Condition=IsValidSoftObjectReference(SoftObjectReference=@n1))
   > SetVisibility(self=Get Image_GameImage(), InVisibility=Visible)
  ]}
 HandleLockedReason {nodes [
   n1 = HandleLockedReason!() -> (then, Reason:EFortPlaylistTravelLockedReason Enum=None) {is=title K2Node_FunctionEntry}
   > n2 = Switch on EFortPlaylistTravelLockedReason!(execute, Selection:EFortPlaylistTravelLockedReason Enum=@n1) -> (None:Exec, Unknown:Exec, NotPartyLeader:Exec, AgeRating:Exec, PartySize:Exec, ActivityLinkNotReady:Exec, AlreadyTraveling:Exec) {is=title K2Node_SwitchEnum}
   SetErrorMessage!(execute=@n2.AgeRating, self, InErrorMessage:Text="LOCTEXT(\"This experience is locked. Your parent or guardian needs to unlock it.\")") -> (then)
   SetErrorMessage(execute=@n2.None)
   SetErrorMessage(execute=@n2.NotPartyLeader, InErrorMessage="LOCTEXT(\"Only the party leader can initiate travel.\")")
   SetErrorMessage(execute=@n2.Unknown, InErrorMessage="LOCTEXT(\"Unknown\")")
   SetErrorMessage(execute=@n2.PartySize, InErrorMessage="LOCTEXT(\"Your current party size it too large for this experience\")")
  ]}
 NewMacro {nodes [
   Select(
    None:Wildcard
    PrepTime:Wildcard
    Launching:Wildcard
    Matchmaking:Wildcard
    MatchmakingSuccess:Wildcard
    MatchmakingFailed:Wildcard
    Index=Inputs!() -> (Index:EFortPlaylistTravelStatus Enum=None)
   ) -> (ReturnValue:Wildcard)
   Outputs()
  ]}
 OnActivityViewModelSet {nodes [
   n1 = OnActivityViewModelSet!() -> (then, ActivityVM:Discover Activity View Model Object Reference) {is=title K2Node_FunctionEntry}
   > Is Valid(InputObject=@n1)
   > SetVisibility(self=Get WBP_UIKit_Throbber() -> (WBP_UIKit_Throbber:WBP UIKit Throbber Object Reference), InVisibility=Collapsed)
   > SetVisibility(self=Get HorizontalBox_Content() -> (HorizontalBox_Content:Horizontal Box Object Reference), InVisibility=SelfHitTestInvisible)
   > SetVisibility(self=Get Image_GameImage(), InVisibility=Collapsed)
   > FetchImages(self:Fort Discover Link Entry Tile Item VM Object Reference=Get ActivityVM() -> (ActivityVM:Discover Activity View Model Object Reference))
   > Call On View Model Data Loaded!(execute:Exec, self:WBP Fort Playlist Travel Tile Object Reference) -> (then:Exec) {is=title K2Node_CallDelegate}
  ]}
 OnTravelButtonClicked {nodes [OnTravelButtonClicked() {is=title K2Node_FunctionEntry}]}
 OnViewModelDataLoaded {nodes [OnViewModelDataLoaded() {is=title K2Node_FunctionEntry}]}
 SetErrorMessage {nodes [
   n1 = SetErrorMessage!() -> (then, InErrorMessage:Text) {is=title K2Node_FunctionEntry}
   > "SetText (Text)"(self=Get Text_Error() -> (Text_Error:Common Text Object Reference), InText=@n1)
  ]}
 SetPlayerCountText {nodes [
   SetPlayerCountText!() -> (then, CurrCount:int) {is=title K2Node_FunctionEntry}
   n1 = Get BP_FortPlaylistTravelEntryViewModel()
   > n2 = Get!(execute:Exec, self:Fort Playlist Travel Entry View Model Object Reference=@n1) -> (then:Exec, else:Exec, ActivityViewModel:Discover Activity View Model Object Reference)
   > SetPlayerCountText!(
    execute
    self:WBP Fort Playlist Travel Player Count Widget Object Reference=Get Widget_PlayerCount() -> (Widget_PlayerCount:WBP Fort Playlist Travel Player Count Widget Object Reference)
    CurrCount:int=GetNumParticipatingPlayers!(self:Fort Playlist Travel Entry View Model Object Reference=@n1) -> (ReturnValue:int=0)
    MaxCount:int=GetMaxTeamSize!(self:Discover Activity View Model Object Reference=@n2) -> (ReturnValue:int=0)
   ) -> (then) {is=title K2Node_CallFunction}
  ]}
 SetTileName {nodes [
   n1 = SetTileName!() -> (then, InText:Text) {is=title K2Node_FunctionEntry}
   > "SetText (Text)"(self=Get Text_Title() -> (Text_Title:Common Text Object Reference), InText=@n1)
  ]}
 UpdateTravelButtonState {nodes [
   n1 = UpdateTravelButtonState!() -> (then, bEnabled:bool=~) {is=title K2Node_FunctionEntry}
   > SetIsEnabled(self=Get Button_Travel(), bInIsEnabled=@n1)
  ]}
]
variables [
 ButtonText_Cancel {type=string, default=Cancel Trip}
 ButtonText_Launching {type=string, default=Launching}
 ButtonText_Travel {type=string, default=Travel Now}
 LaunchingText {type=string, default="Launching!"}
 MatchmakingText {type=string, default=Mathmaking in progress...}
 OnTravelButtonClicked {}
 OnViewModelDataLoaded {}
 PreparingText {type=string, default=Preparing to travel...}
]
widgets {
 info {inheritedWidgetCount=0, parentClass=&CommonUserWidget, rootWidgetClass=&Overlay, widgetCount=24}
 widgets [
  Overlay_13 {widgetClassPath=&Overlay, widget=Overlay_13}
  WBP_UIKit_Backplate {
   widgetClassPath=&WBP_UIKit_Backplate
   bIsVariable=true
   parent=Overlay_13
   widget=WBP_UIKit_Backplate
   properties {brightness=Obscure, cornerRadius=Interactive Component}
   slot {class=OverlaySlot, properties {horizontalAlignment=HAlign_Fill, verticalAlignment=VAlign_Fill}}
  }
  SizeBox_0 {
   widgetClassPath=&SizeBox
   parent=Overlay_13
   widget=SizeBox_0
   properties {bOverride_MaxDesiredHeight=true, bOverride_MinDesiredHeight=true, bOverride_WidthOverride=true, maxDesiredHeight=746, widthOverride=1424}
   slot {class=OverlaySlot, properties {horizontalAlignment=HAlign_Fill, verticalAlignment=VAlign_Fill}}
  }
  Overlay_0 {widgetClassPath=&Overlay, parent=SizeBox_0, widget=Overlay_0, slot {class=SizeBoxSlot}}
  HorizontalBox_Content {
   widgetClassPath=&HorizontalBox
   bIsVariable=true
   parent=Overlay_0
   widget=HorizontalBox_Content
   properties {bOverride_Cursor=true, visibility=Collapsed}
   slot {class=OverlaySlot, properties {horizontalAlignment=HAlign_Fill, verticalAlignment=VAlign_Fill}}
  }
  VerticalBox_2 {
   widgetClassPath=&VerticalBox
   parent=HorizontalBox_Content
   widget=VerticalBox_2
   slot {class=HorizontalBoxSlot, properties {padding=<ltrb 80 80 48 48>, size {sizeRule=Fill, value=1}}}
  }
  Text_Title {
   widgetClassPath=&CommonTextBlock
   bIsVariable=true
   parent=VerticalBox_2
   widget=Text_Title
   properties {
    autoWrapText=true
    text="{Experience Name}"
    font {
     bForceMonospaced=false
     bMaterialIsStencil=false
     fontMaterial=null
     fontObject=&HeadingNow
     letterSpacing=0
     monospacedWidth=1
     size=25.5
     skewAmount=0
     typefaceFontName=86 Bold Italic
     outlineSettings {
      bApplyOutlineToDropShadows=false
      bMiteredCorners=false
      bSeparateFillAlpha=false
      outlineMaterial=null
      outlineSize=0
      outlineColor=<rgba 0 0 0 1>
     }
    }
    shadowOffset=<0 0>
   }
   slot {class=VerticalBoxSlot, properties {padding=<ltrb 0 0 0 10>}}
  }
  HorizontalBox_2 {
   widgetClassPath=&HorizontalBox
   parent=VerticalBox_2
   widget=HorizontalBox_2
   slot {class=VerticalBoxSlot, properties {padding=<ltrb 0 0 0 48>}}
  }
  SqueegeeSlot_Card {
   widgetClassPath=/Game/UI/Plugin/Squeegee/SqueegeeInjectionSlot
   bIsVariable=true
   parent=HorizontalBox_2
   widget=SqueegeeSlot_Card
   properties {
    bIsFocusable=true
    navigation=&self:WidgetTree.SqueegeeSlot_Card.WidgetNavigation_0
    pluginId=Squeegee
    pluginWidgetId=RatingTag
   }
   slot {class=HorizontalBoxSlot, properties {verticalAlignment=VAlign_Top, padding=<ltrb 0 0 5 0>}}
  }
  SqueegeeSlot_Lock {
   widgetClassPath=&SqueegeeInjectionSlot
   bIsVariable=true
   parent=HorizontalBox_2
   widget=SqueegeeSlot_Lock
   properties {pluginId=Squeegee, pluginWidgetId=Lock Icon}
   slot {class=HorizontalBoxSlot, properties {horizontalAlignment=HAlign_Center, verticalAlignment=VAlign_Top}}
  }
  EntryBox_ActivityTags {
   widgetClassPath=&DynamicEntryBox
   bIsVariable=true
   parent=HorizontalBox_2
   widget=EntryBox_ActivityTags
   properties {
    entryBoxType=Wrap
    entryHorizontalAlignment=HAlign_Right
    entryWidgetClass=/Game/UI/DiscoveryShared/ActivityBrowserTag.ActivityBrowserTag_C
    flowDirectionPreference=Culture
    numDesignerPreviewEntries=4
    entrySpacing=<8 8>
   }
   slot {class=HorizontalBoxSlot, properties {padding=<ltrb 5 0 0 0>, size {sizeRule=Fill, value=1}}}
  }
  SizeBox_1 {
   widgetClassPath=&SizeBox
   parent=VerticalBox_2
   widget=SizeBox_1
   properties {bOverride_MaxDesiredHeight=true, bOverride_WidthOverride=true, maxDesiredHeight=238, widthOverride=784}
   slot {class=VerticalBoxSlot}
  }
  ScrollBox_Desc {
   widgetClassPath=&ScrollBox
   bIsVariable=true
   parent=SizeBox_1
   widget=ScrollBox_Desc
   properties {
    scrollbarPadding=<ltrb 20 0 0 0>
    widgetBarStyle {
     thickness=16
     draggedThumbImage {
      bIsDynamicallyLoaded=false
      drawAs=Box
      imageType=FullColor
      mirroring=NoMirror
      resourceName=None
      resourceObject=/FortPlaylistTravelUI/Art/M_UI_FortPlaylistTravel_Circle
      tiling=NoTile
      imageSize=<10 10>
      margin=<ltrb 0.5 0.5 0.5 0.5>
      outlineSettings {
       bUseBrushTransparency=false
       roundingType=HalfHeightRadius
       width=0
       color {colorUseRule=UseColor_Specified, specifiedColor=<rgba 0 0 0 0>}
       cornerRadii=<0 0 0 0>
      }
      tintColor {colorUseRule=UseColor_Specified, specifiedColor=<rgba 1 1 1 1>}
      uVRegion {bIsValid=false, max=<0 0>, min=<0 0>}
     }
     horizontalBackgroundImage {
      bIsDynamicallyLoaded=false
      drawAs=NoDrawType
      imageType=NoImage
      mirroring=NoMirror
      resourceName=None
      resourceObject=null
      tiling=NoTile
      imageSize=<0 0>
      margin=<ltrb 0 0 0 0>
      outlineSettings {
       bUseBrushTransparency=false
       roundingType=HalfHeightRadius
       width=0
       color {colorUseRule=UseColor_Specified, specifiedColor=<rgba 0 0 0 0>}
       cornerRadii=<0 0 0 0>
      }
      tintColor {colorUseRule=UseColor_Specified, specifiedColor=<rgba 1 1 1 1>}
      uVRegion {bIsValid=false, max=<0 0>, min=<0 0>}
     }
     horizontalBottomSlotImage {
      bIsDynamicallyLoaded=false
      drawAs=Image
      imageType=FullColor
      mirroring=NoMirror
      resourceName=../../../Engine/Content/Slate/Common/Scrollbar_Background_Horizontal.png
      resourceObject=null
      tiling=NoTile
      imageSize=<8 8>
      margin=<ltrb 0 0 0 0>
      outlineSettings {
       bUseBrushTransparency=false
       roundingType=HalfHeightRadius
       width=0
       color {colorUseRule=UseColor_Specified, specifiedColor=<rgba 0 0 0 0>}
       cornerRadii=<0 0 0 0>
      }
      tintColor {colorUseRule=UseColor_Specified, specifiedColor=<rgba 1 1 1 1>}
      uVRegion {bIsValid=false, max=<0 0>, min=<0 0>}
     }
     horizontalTopSlotImage {
      bIsDynamicallyLoaded=false
      drawAs=NoDrawType
      imageType=FullColor
      mirroring=NoMirror
      resourceName=../../../Engine/Content/Slate/Common/Scrollbar_Background_Horizontal.png
      resourceObject=null
      tiling=NoTile
      imageSize=<8 8>
      margin=<ltrb 0 0 0 0>
      outlineSettings {
       bUseBrushTransparency=false
       roundingType=HalfHeightRadius
       width=0
       color {colorUseRule=UseColor_Specified, specifiedColor=<rgba 0 0 0 0>}
       cornerRadii=<0 0 0 0>
      }
      tintColor {colorUseRule=UseColor_Specified, specifiedColor=<rgba 1 1 1 1>}
      uVRegion {bIsValid=false, max=<0 0>, min=<0 0>}
     }
     hoveredThumbImage {
      bIsDynamicallyLoaded=false
      drawAs=Box
      imageType=FullColor
      mirroring=NoMirror
      resourceName=None
      resourceObject=&M_UI_FortPlaylistTravel_Circle
      tiling=NoTile
      imageSize=<10 10>
      margin=<ltrb 0.5 0.5 0.5 0.5>
      outlineSettings {
       bUseBrushTransparency=false
       roundingType=HalfHeightRadius
       width=0
       color {colorUseRule=UseColor_Specified, specifiedColor=<rgba 0 0 0 0>}
       cornerRadii=<0 0 0 0>
      }
      tintColor {colorUseRule=UseColor_Specified, specifiedColor=<rgba 1 1 1 1>}
      uVRegion {bIsValid=false, max=<0 0>, min=<0 0>}
     }
     normalThumbImage {
      bIsDynamicallyLoaded=false
      drawAs=Box
      imageType=FullColor
      mirroring=NoMirror
      resourceName=None
      resourceObject=/FortUILibrary/UIKit/ScrollBar/MaterialInstances/MI_UIKit_ScrollbarBackground
      tiling=NoTile
      imageSize=<10 10>
      margin=<ltrb 0.5 0.5 0.5 0.5>
      outlineSettings {
       bUseBrushTransparency=false
       roundingType=HalfHeightRadius
       width=0
       color {colorUseRule=UseColor_Specified, specifiedColor=<rgba 0 0 0 0>}
       cornerRadii=<0 0 0 0>
      }
      tintColor {colorUseRule=UseColor_Specified, specifiedColor=<rgba 1 1 1 1>}
      uVRegion {bIsValid=false, max=<0 0>, min=<0 0>}
     }
     verticalBackgroundImage {
      bIsDynamicallyLoaded=false
      drawAs=Box
      imageType=FullColor
      mirroring=NoMirror
      resourceName=None
      resourceObject=&MI_UIKit_ScrollbarBackground
      tiling=NoTile
      imageSize=<10 10>
      margin=<ltrb 0.5 0.5 0.5 0.5>
      outlineSettings {
       bUseBrushTransparency=false
       roundingType=HalfHeightRadius
       width=0
       color {colorUseRule=UseColor_Specified, specifiedColor=<rgba 0 0 0 0>}
       cornerRadii=<0 0 0 0>
      }
      tintColor {colorUseRule=UseColor_Specified, specifiedColor=<rgba 1 1 1 1>}
      uVRegion {bIsValid=false, max=<0 0>, min=<0 0>}
     }
     verticalBottomSlotImage {
      bIsDynamicallyLoaded=false
      drawAs=NoDrawType
      imageType=FullColor
      mirroring=NoMirror
      resourceName=../../../Engine/Content/Slate/Common/Scrollbar_Background_Vertical.png
      resourceObject=null
      tiling=NoTile
      imageSize=<0 0>
      margin=<ltrb 0 0 0 0>
      outlineSettings {
       bUseBrushTransparency=false
       roundingType=HalfHeightRadius
       width=0
       color {colorUseRule=UseColor_Specified, specifiedColor=<rgba 0 0 0 0>}
       cornerRadii=<0 0 0 0>
      }
      tintColor {colorUseRule=UseColor_Specified, specifiedColor=<rgba 1 1 1 1>}
      uVRegion {bIsValid=false, max=<0 0>, min=<0 0>}
     }
     verticalTopSlotImage {
      bIsDynamicallyLoaded=false
      drawAs=NoDrawType
      imageType=FullColor
      mirroring=NoMirror
      resourceName=../../../Engine/Content/Slate/Common/Scrollbar_Background_Vertical.png
      resourceObject=null
      tiling=NoTile
      imageSize=<8 8>
      margin=<ltrb 0 0 0 0>
      outlineSettings {
       bUseBrushTransparency=false
       roundingType=HalfHeightRadius
       width=0
       color {colorUseRule=UseColor_Specified, specifiedColor=<rgba 0 0 0 0>}
       cornerRadii=<0 0 0 0>
      }
      tintColor {colorUseRule=UseColor_Specified, specifiedColor=<rgba 1 1 1 1>}
      uVRegion {bIsValid=false, max=<0 0>, min=<0 0>}
     }
    }
    widgetStyle {
     barThickness=9
     bottomShadowBrush {
      bIsDynamicallyLoaded=false
      drawAs=Box
      imageType=FullColor
      mirroring=NoMirror
      resourceName=../../../Engine/Content/Slate/Common/ScrollBoxShadowBottom.png
      resourceObject=null
      tiling=NoTile
      imageSize=<24 8>
      margin=<ltrb 0.5 0 0.5 1>
      outlineSettings {
       bUseBrushTransparency=false
       roundingType=HalfHeightRadius
       width=0
       color {colorUseRule=UseColor_Specified, specifiedColor=<rgba 0 0 0 0>}
       cornerRadii=<0 0 0 0>
      }
      tintColor {colorUseRule=UseColor_Specified, specifiedColor=<rgba 1 1 1 1>}
      uVRegion {bIsValid=false, max=<0 0>, min=<0 0>}
     }
     horizontalScrolledContentPadding=<ltrb 0 0 1 0>
     leftShadowBrush {
      bIsDynamicallyLoaded=false
      drawAs=Box
      imageType=FullColor
      mirroring=NoMirror
      resourceName=../../../Engine/Content/Slate/Common/ScrollBoxShadowLeft.png
      resourceObject=null
      tiling=NoTile
      imageSize=<8 16>
      margin=<ltrb 1 0.5 0 0.5>
      outlineSettings {
       bUseBrushTransparency=false
       roundingType=HalfHeightRadius
       width=0
       color {colorUseRule=UseColor_Specified, specifiedColor=<rgba 0 0 0 0>}
       cornerRadii=<0 0 0 0>
      }
      tintColor {colorUseRule=UseColor_Specified, specifiedColor=<rgba 1 1 1 1>}
      uVRegion {bIsValid=false, max=<0 0>, min=<0 0>}
     }
     rightShadowBrush {
      bIsDynamicallyLoaded=false
      drawAs=Box
      imageType=FullColor
      mirroring=NoMirror
      resourceName=../../../Engine/Content/Slate/Common/ScrollBoxShadowRight.png
      resourceObject=null
      tiling=NoTile
      imageSize=<8 16>
      margin=<ltrb 0 0.5 1 0.5>
      outlineSettings {
       bUseBrushTransparency=false
       roundingType=HalfHeightRadius
       width=0
       color {colorUseRule=UseColor_Specified, specifiedColor=<rgba 0 0 0 0>}
       cornerRadii=<0 0 0 0>
      }
      tintColor {colorUseRule=UseColor_Specified, specifiedColor=<rgba 1 1 1 1>}
      uVRegion {bIsValid=false, max=<0 0>, min=<0 0>}
     }
     topShadowBrush {
      bIsDynamicallyLoaded=false
      drawAs=Box
      imageType=FullColor
      mirroring=NoMirror
      resourceName=../../../Engine/Content/Slate/Common/ScrollBoxShadowTop.png
      resourceObject=null
      tiling=NoTile
      imageSize=<24 8>
      margin=<ltrb 0.5 1 0.5 0>
      outlineSettings {
       bUseBrushTransparency=false
       roundingType=HalfHeightRadius
       width=0
       color {colorUseRule=UseColor_Specified, specifiedColor=<rgba 0 0 0 0>}
       cornerRadii=<0 0 0 0>
      }
      tintColor {colorUseRule=UseColor_Specified, specifiedColor=<rgba 1 1 1 1>}
      uVRegion {bIsValid=false, max=<0 0>, min=<0 0>}
     }
     verticalScrolledContentPadding=<ltrb 0 0 0 1>
    }
   }
   slot {class=SizeBoxSlot}
  }
  Text_Desc {
   widgetClassPath=&CommonTextBlock
   bIsVariable=true
   parent=ScrollBox_Desc
   widget=Text_Desc
   properties {
    autoWrapText=true
    lineHeightPercentage=2
    text="{Experience description text with some extra here to show wrapping and see if we can make it bigerCatasstrophe meow meow pee in shoe. Leave hair on owner's clothes meow to be let in for hide from vacuum cleaner but cuddle in your lap.}"
    font {
     bForceMonospaced=false
     bMaterialIsStencil=false
     fontMaterial=null
     fontObject=&HeadingNow
     letterSpacing=0
     monospacedWidth=1
     size=18
     skewAmount=0
     typefaceFontName=74 Regular
     outlineSettings {
      bApplyOutlineToDropShadows=false
      bMiteredCorners=false
      bSeparateFillAlpha=false
      outlineMaterial=null
      outlineSize=0
      outlineColor=<rgba 0 0 0 1>
     }
    }
    shadowOffset=<0 0>
   }
   slot {class=ScrollBoxSlot}
  }
  HorizontalBox_TravelState {
   widgetClassPath=&HorizontalBox
   bIsVariable=true
   parent=VerticalBox_2
   widget=HorizontalBox_TravelState
   slot {class=VerticalBoxSlot, properties {verticalAlignment=VAlign_Bottom, padding=<ltrb 0 40 0 0>, size {sizeRule=Fill, value=1}}}
  }
  Text_Timer {
   widgetClassPath=&CommonTextBlock
   bIsVariable=true
   parent=HorizontalBox_TravelState
   widget=Text_Timer
   properties {
    text="{0:00}"
    font {
     bForceMonospaced=false
     bMaterialIsStencil=false
     fontMaterial=null
     fontObject=&HeadingNow
     letterSpacing=0
     monospacedWidth=1
     size=15.75
     skewAmount=0
     typefaceFontName=75 Medium
     outlineSettings {
      bApplyOutlineToDropShadows=false
      bMiteredCorners=false
      bSeparateFillAlpha=false
      outlineMaterial=null
      outlineSize=0
      outlineColor=<rgba 0 0 0 1>
     }
    }
    shadowOffset=<0 0>
   }
   slot {class=HorizontalBoxSlot}
  }
  Text_TravelState {
   widgetClassPath=&CommonTextBlock
   bIsVariable=true
   parent=HorizontalBox_TravelState
   widget=Text_TravelState
   properties {
    text="{Travel State}"
    colorAndOpacity {colorUseRule=UseColor_Specified, specifiedColor=<rgba 1 1 1 0.7>}
    font {
     bForceMonospaced=false
     bMaterialIsStencil=false
     fontMaterial=null
     fontObject=&HeadingNow
     letterSpacing=0
     monospacedWidth=1
     size=15.75
     skewAmount=0
     typefaceFontName=74 Regular
     outlineSettings {
      bApplyOutlineToDropShadows=false
      bMiteredCorners=false
      bSeparateFillAlpha=false
      outlineMaterial=null
      outlineSize=0
      outlineColor=<rgba 0 0 0 1>
     }
    }
    shadowOffset=<0 0>
   }
   slot {class=HorizontalBoxSlot, properties {padding=<ltrb 5 0 0 0>}}
  }
  HorizontalBox_1 {
   widgetClassPath=&HorizontalBox
   parent=VerticalBox_2
   widget=HorizontalBox_1
   slot {class=VerticalBoxSlot, properties {verticalAlignment=VAlign_Bottom, padding=<ltrb 0 20 0 0>}}
  }
  Button_Travel {
   widgetClassPath=/FortUILibrary/UIKit/Buttons/WBP_UIKit_Button_Loud
   bIsVariable=true
   parent=HorizontalBox_1
   widget=Button_Travel
   properties {
    bIsFocusable=false
    text="{Travel Button Text}"
    useDisabledVisualsOnLocked=true
    visibility=Visible
    triggeringInputAction {dataTable=&/FortPlaylistTravelUIActions, rowName=Travel}
   }
   slot {class=HorizontalBoxSlot, properties {horizontalAlignment=HAlign_Left, verticalAlignment=VAlign_Center, size {sizeRule=Fill, value=1}}}
  }
  Widget_PlayerCount {
   widgetClassPath=&/WBP_FortPlaylistTravel_PlayerCountWidget
   bIsVariable=true
   parent=HorizontalBox_1
   widget=Widget_PlayerCount
   slot {class=HorizontalBoxSlot, properties {verticalAlignment=VAlign_Center}}
  }
  Spacer_96 {widgetClassPath=&Spacer, parent=HorizontalBox_1, widget=Spacer_96, properties {size=<1 40>}, slot {class=HorizontalBoxSlot}}
  Text_Error {
   widgetClassPath=&CommonTextBlock
   bIsVariable=true
   parent=VerticalBox_2
   widget=Text_Error
   properties {
    autoWrapText=true
    bAutoCollapseWithEmptyText=true
    text="{Error Text Showing It Wrapping to Test So make it longer} "
    colorAndOpacity {colorUseRule=UseColor_Specified, specifiedColor=<rgba 1 0.147 0.1329 1>}
    font {
     bForceMonospaced=false
     bMaterialIsStencil=false
     fontMaterial=null
     fontObject=&HeadingNow
     letterSpacing=0
     monospacedWidth=1
     size=13.5
     skewAmount=0
     typefaceFontName=74 Regular
     outlineSettings {
      bApplyOutlineToDropShadows=false
      bMiteredCorners=false
      bSeparateFillAlpha=false
      outlineMaterial=null
      outlineSize=0
      outlineColor=<rgba 0 0 0 1>
     }
    }
    shadowOffset=<0 0>
   }
   slot {class=VerticalBoxSlot, properties {verticalAlignment=VAlign_Bottom, padding=<ltrb 0 32 0 0>}}
  }
  Image_GameImage {
   widgetClassPath=&FortLazyImage
   bIsVariable=true
   parent=HorizontalBox_Content
   widget=Image_GameImage
   properties {
    bOverride_Cursor=true
    materialTextureParamName=Image
    visibility=Collapsed
    brush {
     bIsDynamicallyLoaded=false
     drawAs=Box
     imageType=FullColor
     mirroring=NoMirror
     resourceName=None
     resourceObject=/FortPlaylistTravelUI/Art/MI_UI_FortPlaylistTravel_GamePreviewImage
     tiling=NoTile
     imageSize=<512 512>
     margin=<ltrb 0.5 0.5 0.5 0.5>
     outlineSettings {
      bUseBrushTransparency=false
      roundingType=HalfHeightRadius
      width=0
      color {colorUseRule=UseColor_Specified, specifiedColor=<rgba 0 0 0 0>}
      cornerRadii=<0 0 0 0>
     }
     tintColor {colorUseRule=UseColor_Specified, specifiedColor=<rgba 1 1 1 1>}
     uVRegion {bIsValid=false, max=<0 0>, min=<0 0>}
    }
   }
   slot {class=HorizontalBoxSlot, properties {size {sizeRule=Automatic, value=0.5}}}
  }
  WBP_UIKit_Throbber {
   widgetClassPath=/FortUILibrary/UIKit/Throbber/WBP_UIKit_Throbber
   bIsVariable=true
   parent=Overlay_13
   widget=WBP_UIKit_Throbber
   properties {displayBackground=false, throbberSize=Large}
   slot {class=OverlaySlot, properties {horizontalAlignment=HAlign_Center, verticalAlignment=VAlign_Center}}
  }
 ]
}
