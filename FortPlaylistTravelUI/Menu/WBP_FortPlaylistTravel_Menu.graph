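# Activatable "Experience Travel" menu where players page through playlist destinations and confirm or cancel game travel.
# SetListItems builds one carousel tile per entry view model, then selects the current destination a tick later.
# Tile clicks open a Yes/No modal scene that starts travel, or cancels it during PrepTime, through the travel view model.
# Page changes recolor the travel icon material by travel eligibility; Next and Prev button handlers call opposite page functions.
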WidgetBlueprint /FortPlaylistTravelUI/Menu/WBP_FortPlaylistTravel_Menu
animations [
 Anim_ModalOverlay {
  binding_count=1
  length_seconds=0.2
  object=&self:Anim_ModalOverlay
  sequence {bindings [ModalOverlay {
     id=#1
     display_name=WBP_UIKit_Backplate_1
     possessed_class=WBP_UIKit_Backplate_C
     tracks [{
       type=MovieSceneFloatTrack
       display_name=Render Opacity
       sections [{type=MovieSceneFloatSection, channels [None {type=MovieSceneScriptingFloatChannel, keys [[0,0],[4,1]]}]}]
      }]
    }]}
 }
 Intro {
  binding_count=5
  length_seconds=0.34
  object=&self:Intro
  sequence {
   bindings [
    Button_NextPage {
     id=#2
     possessed_class=WBP_UIKit_Button_Generic_C
     tracks [{
       type=MovieSceneFloatTrack
       display_name=Render Opacity
       sections [{type=MovieSceneFloatSection, channels [None {type=MovieSceneScriptingFloatChannel, keys [[16,0],[28,1]]}]}]
      }]
    }
    Overlay_2 {
     id=#3
     possessed_class=Overlay
     tracks [
      {
       type=MovieScene2DTransformTrack
       sections [{
         type=MovieScene2DTransformSection
         channels [
          Translation.X {type=MovieSceneScriptingFloatChannel, keys [[0,-300],[25,0]]}
          Translation.Y {type=MovieSceneScriptingFloatChannel, const [0,2]}
          Angle {type=MovieSceneScriptingFloatChannel, default=0}
          Scale.X {type=MovieSceneScriptingFloatChannel, default=1}
          Scale.Y {type=MovieSceneScriptingFloatChannel, default=1}
          Shear.X {type=MovieSceneScriptingFloatChannel, default=0}
          Shear.Y {type=MovieSceneScriptingFloatChannel, default=0}
         ]
        }]
      }
      {
       type=MovieSceneFloatTrack
       display_name=Render Opacity
       sections [{type=MovieSceneFloatSection, channels [None {type=MovieSceneScriptingFloatChannel, keys [[0,0],[25,1]]}]}]
      }
     ]
    }
    Button_Close {
     id=#4
     possessed_class=WBP_UIKit_Button_Quiet_C
     tracks [{
       type=MovieSceneFloatTrack
       display_name=Render Opacity
       sections [{type=MovieSceneFloatSection, channels [None {type=MovieSceneScriptingFloatChannel, keys [[22,0],[34,1]]}]}]
      }]
    }
    Button_PrevPage {
     id=#5
     possessed_class=WBP_UIKit_Button_Generic_C
     tracks [{
       type=MovieSceneFloatTrack
       display_name=Render Opacity
       sections [{type=MovieSceneFloatSection, channels [None {type=MovieSceneScriptingFloatChannel, keys [[16,0],[28,1]]}]}]
      }]
    }
    Overlay_1 {
     id=#6
     possessed_class=Overlay
     tracks [{
       type=MovieSceneFloatTrack
       display_name=Render Opacity
       sections [{type=MovieSceneFloatSection, channels [None {type=MovieSceneScriptingFloatChannel, keys [[0,0],[25,1]]}]}]
      }]
    }
   ]
   display_rate {numerator=100}
  }
 }
]
graphs [
 CancelTravel {nodes [
   CancelTravel() {is=title K2Node_FunctionEntry}
   > CancelGameTravel!(
    execute
    self:Fort Playlist Travel View Model Object Reference=Get BP_FortPlaylistTravelViewModel() -> (BP_FortPlaylistTravelViewModel:BP Fort Playlist Travel View Model Object Reference)
    bCancelIndividualOnly:bool=false
   ) -> (then)
  ]}
 EventGraph {nodes [
   "On Clicked (Button_NextPage)"!() -> (OutputDelegate, then, Button:Common Button Base Object Reference) {is=title K2Node_ComponentBoundEvent}
   > PreviousPage(self:Common Widget Carousel Object Reference=Get ContentCarousel() -> (ContentCarousel:Common Widget Carousel Object Reference))
   n1 = "On Current Page Index Changed (ContentCarousel)"!() -> (OutputDelegate, then, CarouselWidget:Common Widget Carousel Object Reference, CurrentPageIndex:int=0) {is=title K2Node_ComponentBoundEvent}
   > DelayUntilNextTick!(execute) -> (then)
   > n2 = Cast To WBP_FortPlaylistTravel_Tile!(
    execute
    Object=GetChildAt!(self:Panel Widget Object Reference=@n1.CarouselWidget, Index:int=@n1.CurrentPageIndex) -> (ReturnValue:Widget Object Reference)
   ) -> (then, CastFailed, AsWBP Fort Playlist Travel Tile:WBP Fort Playlist Travel Tile Object Reference) {is=title K2Node_DynamicCast}
   n3 = Get ScrollBox_Desc!(self:WBP Fort Playlist Travel Tile Object Reference=@n2) -> (ScrollBox_Desc:Scroll Box Object Reference)
   > ScrollToStart(self:Scroll Box Object Reference=@n3)
   > RegisterScrollRecipientExternal!(execute, self:Common User Widget Object Reference, AnalogScrollRecipient:Widget Object Reference=@n3) -> (then)
   > n4 = Get!(execute:Exec, self:WBP Fort Playlist Travel Tile Object Reference=@n2) -> (then:Exec, else:Exec, BP_FortPlaylistTravelEntryViewModel:BP Fort Playlist Travel Entry View Model Object Reference)
   > SetTravelIconState!(
    execute
    self
    bCanTravel:bool=Get CanInitiateGameTravel!(self:Fort Playlist Travel Entry View Model Object Reference=@n4) -> (CanInitiateGameTravel:bool)
   ) -> (then)
   SetTravelIconState(execute=@n4.else, bCanTravel=true)
   n5 = "On Clicked (Button_Close)"!() -> (OutputDelegate, then, Button:Common Button Base Object Reference) {is=title K2Node_ComponentBoundEvent}
   DeactivateWidget(execute=[@n5.then, @n6.then])
   "On Clicked (Button_PrevPage)"!() -> (OutputDelegate, then, Button:Common Button Base Object Reference) {is=title K2Node_ComponentBoundEvent}
   > NextPage(self:Common Widget Carousel Object Reference=Get ContentCarousel())
   n6 = "On Clicked (CloseButton_Touch)"() {is=title K2Node_ComponentBoundEvent}
   OnConfirmClicked() {is=title K2Node_CustomEvent}
   > RemoveModal()
   > n7 = Branch(Condition=IsPreparingToLaunch!() -> (ReturnValue:bool))
   > CancelTravel() {is=title K2Node_CallFunction}
   HandleTravel(execute=@n7.else)
   OnCancelClicked() {is=title K2Node_CustomEvent}
   > RemoveModal()
   SetCurrentActiveIndex() {is=title K2Node_CustomEvent}
   > DelayUntilNextTick()
   > SetActiveWidgetIndex(self:Common Widget Carousel Object Reference=Get ContentCarousel(), Index=Get CurrentTravelIndex() -> (CurrentTravelIndex:int=0))
   Event Construct()
   > PlayAnimationForward(InAnimation=Get Intro())
   > SetLinkedCarousel!(
    execute
    self:Common Widget Carousel Nav Bar Object Reference=Get ContentCarousel_Nav() -> (ContentCarousel_Nav:Common Widget Carousel Nav Bar Object Reference)
    CommonCarousel:Common Widget Carousel Object Reference=Get ContentCarousel()
   ) -> (then)
  ]}
 HandleTileViewmodelDataLoaded {nodes [
   HandleTileViewmodelDataLoaded() {is=title K2Node_FunctionEntry}
   n1 = Get ContentCarousel()
   > n2 = Cast To WBP_FortPlaylistTravel_Tile(Object=GetChildAt(self=@n1, Index=GetActiveWidgetIndex!(self:Common Widget Carousel Object Reference=@n1) -> (ReturnValue:int=0)))
   > n3 = Get(self=@n2)
   > SetTravelIconState(bCanTravel=Get CanInitiateGameTravel(self=@n3))
   SetTravelIconState(execute=@n3.else, bCanTravel=true)
  ]}
 HandleTravel {nodes [
   HandleTravel() {is=title K2Node_FunctionEntry}
   n1 = Get ContentCarousel()
   > n2 = Cast To WBP_FortPlaylistTravel_Tile(Object=GetWidgetAtIndex!(self:Common Widget Carousel Object Reference=@n1, Index:int=GetActiveWidgetIndex(self=@n1) -> (ReturnValue=0)) -> (ReturnValue:Widget Object Reference))
   > InitiateGameTravel!(
    execute
    self:Fort Playlist Travel View Model Object Reference=Get BP_FortPlaylistTravelViewModel()
    GameTravelEntryViewModel:Fort Playlist Travel Entry View Model Object Reference=Get BP_FortPlaylistTravelEntryViewModel!(self:WBP Fort Playlist Travel Tile Object Reference=@n2) -> (BP_FortPlaylistTravelEntryViewModel:BP Fort Playlist Travel Entry View Model Object Reference)
   ) -> (then)
   > DeactivateWidget()
  ]}
 HandleTravelButtonClicked {nodes [
   HandleTravelButtonClicked() {is=title K2Node_FunctionEntry}
   > n1 = Branch(Condition=IsPreparingToLaunch())
   > ShowCancelConfirmation()
   ShowTravelConfirmation(execute=@n1.else)
  ]}
 IsPreparingToLaunch {nodes [
   IsPreparingToLaunch() {is=title K2Node_FunctionEntry}
   n1 = Get BP_FortPlaylistTravelViewModel()
   > Return Node!(
    execute
    ReturnValue:bool="Equal (Enum)"(
     A:EFortPlaylistTravelStatus Enum=GetGameTravelStatus!(self:Fort Playlist Travel View Model Object Reference=@n1) -> (ReturnValue:EFortPlaylistTravelStatus Enum=None)
     B:EFortPlaylistTravelStatus Enum=PrepTime
    )
   )
  ]}
 MakeCancelModalContext {nodes [
   MakeCancelModalContext() {is=title K2Node_FunctionEntry}
   > n1 = Construct Object from Class(Class=/Script/FortPlaylistTravelUI.FortPlaylistTravelUIModalSelectionContext) -> (ReturnValue:Fort Playlist Travel UIModal Selection Context Object Reference)
   n2 = Get ActivityViewModel!(self:Fort Playlist Travel Entry View Model Object Reference=Get CurrentGameTravelDestinationViewModel!(self:Fort Playlist Travel View Model Object Reference=Get BP_FortPlaylistTravelViewModel()) -> (CurrentGameTravelDestinationViewModel:Fort Playlist Travel Entry View Model Object Reference)) -> (ActivityViewModel:Discover Activity View Model Object Reference)
   > Set Description!(
    execute
    Description:Text=Format Text!(
     Format="LOCTEXT(\"Are you sure you want to cancel traveling to <emphasis>{Destination}</>?\")"
     Destination:Text=GetDisplayName!(self:Fort Discover Link Entry Tile Item VM Object Reference=@n2) -> (ReturnValue:Text)
    ) -> (Result)
    self:Fort Playlist Travel UIModal Selection Context Object Reference=@n1
   ) -> (then, Output_Get:Text)
   > AddSelectionChoice!(
    execute
    self:Fort Playlist Travel UIModal Selection Context Object Reference=@n1
    SelectionTitle:Text="LOCTEXT(\"Yes\")"
    Action:Data Table Row Handle Structure=Make Data Table Row Handle!(DataTable:Data Table Object Reference=&/FortPlaylistTravelUIActions, RowName:Name=Modal_Confirm) -> (DataTableRowHandle:Data Table Row Handle Structure) {is=title K2Node_MakeStruct}
    SelectionDelegate:Delegate=Create Event()
   ) -> (then)
   > AddSelectionChoice(
    self=@n1
    SelectionTitle="LOCTEXT(\"No\")"
    Action=Make Data Table Row Handle(DataTable=&FortPlaylistTravelUIActions, RowName=Modal_Cancel)
    SelectionDelegate=Create Event()
   )
   > Set CancelContext(CancelContext:Fort Playlist Travel UIModal Selection Context Object Reference=@n1) -> (Output_Get:Fort Playlist Travel UIModal Selection Context Object Reference)
  ]}
 MakeTravelModalContext {nodes [
   MakeTravelModalContext() {is=title K2Node_FunctionEntry}
   n1 = Get ContentCarousel()
   > n2 = Cast To WBP_FortPlaylistTravel_Tile(Object=GetWidgetAtIndex(self=@n1, Index=GetActiveWidgetIndex(self=@n1) -> (ReturnValue=0)))
   > n3 = Construct Object from Class(Class=&FortPlaylistTravelUIModalSelectionContext)
   n4 = Get ActivityViewModel(self=Get BP_FortPlaylistTravelEntryViewModel(self=@n2))
   > Set Description(
    Description=Format Text(
     Format="LOCTEXT(\"Are you sure you want to travel to <emphasis>{Destination}</>?\\r\\nThis will teleport you out of your current game mode.\")"
     Destination=GetDisplayName(self=@n4)
    )
    self=@n3
   )
   > AddSelectionChoice(
    self=@n3
    SelectionTitle="LOCTEXT(\"Yes\")"
    Action=Make Data Table Row Handle(DataTable=&FortPlaylistTravelUIActions, RowName=Modal_Confirm)
    SelectionDelegate=Create Event()
   )
   > AddSelectionChoice(
    self=@n3
    SelectionTitle="LOCTEXT(\"No\")"
    Action=Make Data Table Row Handle(DataTable=&FortPlaylistTravelUIActions, RowName=Modal_Cancel)
    SelectionDelegate=Create Event()
   )
   > Set ConfirmationContext(ConfirmationContext:Fort Playlist Travel UIModal Selection Context Object Reference=@n3) -> (Output_Get:Fort Playlist Travel UIModal Selection Context Object Reference)
  ]}
 RemoveModal {nodes [
   RemoveModal() {is=title K2Node_FunctionEntry}
   > RemoveModalScene!(execute, PlayerController:Player Controller Object Reference=GetOwningPlayer()) -> (then)
   > SetVisibility(self=Get ModalOverlay() -> (ModalOverlay:WBP UIKit Backplate Object Reference), InVisibility=Collapsed)
   > SetVisibility(self=Get Button_Close() -> (Button_Close:WBP UIKit Button Quiet Object Reference), InVisibility=Visible)
  ]}
 SetListItems {nodes [
   n1 = SetListItems!() -> (then, EntryViewModels:Array of Fort Playlist Travel Entry View Model Object References) {is=title K2Node_FunctionEntry}
   > ClearChildren(self=Get ContentCarousel())
   > n2 = For Each Loop(Array:Array of Fort Playlist Travel Entry View Model Object References=@n1) -> (Array Element:Fort Playlist Travel Entry View Model Object Reference)
   > n3 = Cast To BP_FortPlaylistTravelEntryViewModel!(execute, Object=@n2) -> (then, CastFailed, AsBP Fort Playlist Travel Entry View Model:BP Fort Playlist Travel Entry View Model Object Reference) {is=title K2Node_DynamicCast}
   > n4 = Create Widget!(
    execute:Exec
    Class:User Widget Class Reference=&/WBP_FortPlaylistTravel_Tile.WBP_FortPlaylistTravel_Tile_C
    OwningPlayer:Player Controller Object Reference=GetOwningPlayer()
    BP_FortPlaylistTravelEntryViewModel:BP Fort Playlist Travel Entry View Model Object Reference=@n3
   ) -> (then:Exec, ReturnValue:WBP Fort Playlist Travel Tile Object Reference)
   > Bind Event to On Travel Button Clicked(self:WBP Fort Playlist Travel Tile Object Reference=@n4, Delegate=Create Event()) {is=title K2Node_AddDelegate}
   > Bind Event to On View Model Data Loaded(self:WBP Fort Playlist Travel Tile Object Reference=@n4, Delegate=Create Event()) {is=title K2Node_AddDelegate}
   > AddChild(self=Get ContentCarousel(), Content=@n4)
   n5 = Get BP_FortPlaylistTravelViewModel()
   > Branch(Condition="Equal (Object)"(A=Get CurrentGameTravelDestinationViewModel(self=@n5), B=@n3))
   > Set CurrentTravelIndex(CurrentTravelIndex:int=@n2.Array Index) -> (Output_Get:int=0)
   n6 = Length(TargetArray:Array of Fort Playlist Travel Entry View Model Object References=@n1)
   SetVisibility(
    execute=@n2.Completed
    self=[
     Get Button_PrevPage() -> (Button_PrevPage:WBP UIKit Button Generic Object Reference)
     Get Button_NextPage() -> (Button_NextPage:WBP UIKit Button Generic Object Reference)
     Get ContentCarousel_Nav()
    ]
    InVisibility=Select(Option 0=Collapsed, Option 1=Visible, Index="integer > integer"(A=@n6, B=1))
   )
   > SetCurrentActiveIndex() {is=title K2Node_CallFunction}
  ]}
 SetTravelIconState {nodes [
   n1 = SetTravelIconState!() -> (then, bCanTravel:bool=true) {is=title K2Node_FunctionEntry}
   > n2 = GetDynamicMaterial(self=Get Image_TravelIcon() -> (Image_TravelIcon:Image Object Reference))
   > SetVectorParameterValue(self=@n2, ParameterName=BackingRGB, Value=SelectColor(A=(R=0.0395,G=1,B=0.2307,A=1), B=(R=1,G=0.147,B=0.1329,A=1), bPickA=@n1))
   > SetScalarParameterValue(self=@n2, ParameterName=AnimationActive, Value=SelectFloat(A=1, B=0, bPickA=@n1))
  ]}
 ShowCancelConfirmation {nodes [
   ShowCancelConfirmation() {is=title K2Node_FunctionEntry}
   > MakeCancelModalContext() {is=title K2Node_CallFunction}
   > AddModalScene!(
    execute
    PlayerController:Player Controller Object Reference=GetOwningPlayer()
    ModalScene:"Fort Playlist Travel UIModal Scene Operation Structure (by ref)"=Make Fort Playlist Travel UIModal Scene Operation!(
     ModalScene:Dynamic UIScene Weak object=/FortPlaylistTravelUI/Scenes/FortPlaylistTravelModalSelectionScene
     Payload:Object Weak object=Get CancelContext() -> (CancelContext:Fort Playlist Travel UIModal Selection Context Object Reference)
    ) -> (FortPlaylistTravelUIModalSceneOperation:Fort Playlist Travel UIModal Scene Operation Structure) {is=title K2Node_MakeStruct}
   ) -> (then)
   > SetVisibility(self=Get ModalOverlay(), InVisibility=Visible)
   > SetVisibility(self=Get Button_Close(), InVisibility=Hidden)
   > PlayAnimationForward(InAnimation=Get Anim_ModalOverlay() -> (Anim_ModalOverlay:Widget Animation Object Reference))
  ]}
 ShowTravelConfirmation {nodes [
   ShowTravelConfirmation() {is=title K2Node_FunctionEntry}
   > MakeTravelModalContext() {is=title K2Node_CallFunction}
   > AddModalScene(
    PlayerController=GetOwningPlayer()
    ModalScene=Make Fort Playlist Travel UIModal Scene Operation(
     ModalScene=&FortPlaylistTravelModalSelectionScene
     Payload=Get ConfirmationContext() -> (ConfirmationContext:Fort Playlist Travel UIModal Selection Context Object Reference)
    )
   )
   > SetVisibility(self=Get ModalOverlay(), InVisibility=Visible)
   > SetVisibility(self=Get Button_Close(), InVisibility=Hidden)
   > PlayAnimationForward(InAnimation=Get Anim_ModalOverlay())
  ]}
]
variables [
 CancelContext {type=FortPlaylistTravelUIModalSelectionContext, default=null}
 ConfirmationContext {type=FortPlaylistTravelUIModalSelectionContext, default=null}
 CurrentTravelIndex {type=integer}
]
widgets {
 info {inheritedWidgetCount=0, parentClass=&CommonActivatableWidget, rootWidgetClass=&Overlay, widgetCount=19}
 widgets [
  Overlay_0 {widgetClassPath=&Overlay, widget=Overlay_0}
  CHVB_Touch-Only {
   widgetClassPath=&CommonHardwareVisibilityBorder
   parent=Overlay_0
   widget=CHVB_Touch-Only
   properties {visibilityQuery {
     autoDescription=" ANY( Input.Touch )"
     tokenStreamVersion=0
     userDescription=Touch-Only
     queryTokenStream [0,1,1,1,0]
     tagDictionary [{tagName=Input.Touch}]
    }}
   slot {class=OverlaySlot, properties {horizontalAlignment=HAlign_Fill, verticalAlignment=VAlign_Fill}}
  }
  CloseButton_Touch {
   widgetClassPath=&Button
   parent=CHVB_Touch-Only
   widget=CloseButton_Touch
   properties {
    backgroundColor=<rgba 1 1 1 0>
    colorAndOpacity=<rgba 1 1 1 0>
    widgetStyle {
     clickedSlateSound {resourceObject=null}
     disabled {
      bIsDynamicallyLoaded=false
      drawAs=NoDrawType
      imageType=NoImage
      mirroring=NoMirror
      resourceName=None
      resourceObject=null
      tiling=NoTile
      imageSize=<0 0>
      margin=<ltrb 0 0 0 0>
      outlineSettings {
       bUseBrushTransparency=false
       roundingType=HalfHeightRadius
       width=0
       color {colorUseRule=UseColor_Specified, specifiedColor=<rgba 0 0 0 0>}
       cornerRadii=<0 0 0 0>
      }
      tintColor {colorUseRule=UseColor_Specified, specifiedColor=<rgba 1 1 1 1>}
      uVRegion {bIsValid=false, max=<0 0>, min=<0 0>}
     }
     disabledForeground {colorUseRule=UseColor_Specified, specifiedColor=<rgba 0.5271 0.5271 0.5271 1>}
     hovered {
      bIsDynamicallyLoaded=false
      drawAs=NoDrawType
      imageType=NoImage
      mirroring=NoMirror
      resourceName=None
      resourceObject=null
      tiling=NoTile
      imageSize=<32 32>
      margin=<ltrb 0 0 0 0>
      outlineSettings {
       bUseBrushTransparency=true
       roundingType=FixedRadius
       width=1
       color {colorUseRule=UseColor_Specified, specifiedColor=<rgba 0.7243 0.7243 0.7243 1>}
       cornerRadii=<4 4 4 4>
      }
      tintColor {colorUseRule=UseColor_Specified, specifiedColor=<rgba 0.7243 0.7243 0.7243 1>}
      uVRegion {bIsValid=false, max=<0 0>, min=<0 0>}
     }
     hoveredForeground {colorUseRule=UseColor_Specified, specifiedColor=<rgba 0 1 1 1>}
     hoveredSlateSound {resourceObject=null}
     normal {
      bIsDynamicallyLoaded=false
      drawAs=NoDrawType
      imageType=NoImage
      mirroring=NoMirror
      resourceName=None
      resourceObject=null
      tiling=NoTile
      imageSize=<32 32>
      margin=<ltrb 0 0 0 0>
      outlineSettings {
       bUseBrushTransparency=true
       roundingType=FixedRadius
       width=1
       color {colorUseRule=UseColor_Specified, specifiedColor=<rgba 0.6951 0.6951 0.6951 1>}
       cornerRadii=<4 4 4 4>
      }
      tintColor {colorUseRule=UseColor_Specified, specifiedColor=<rgba 0.4955 0.4955 0.4955 1>}
      uVRegion {bIsValid=false, max=<0 0>, min=<0 0>}
     }
     normalForeground {colorUseRule=UseColor_Specified, specifiedColor=<rgba 0 1 1 1>}
     normalPadding=<ltrb 0 0 0 0>
     pressed {
      bIsDynamicallyLoaded=false
      drawAs=NoDrawType
      imageType=NoImage
      mirroring=NoMirror
      resourceName=None
      resourceObject=null
      tiling=NoTile
      imageSize=<32 32>
      margin=<ltrb 0 0 0 0>
      outlineSettings {
       bUseBrushTransparency=true
       roundingType=FixedRadius
       width=1
       color {colorUseRule=UseColor_Specified, specifiedColor=<rgba 0.7243 0.7243 0.7243 1>}
       cornerRadii=<4 4 4 4>
      }
      tintColor {colorUseRule=UseColor_Specified, specifiedColor=<rgba 0.3843 0.3843 0.3843 1>}
      uVRegion {bIsValid=false, max=<0 0>, min=<0 0>}
     }
     pressedForeground {colorUseRule=UseColor_Specified, specifiedColor=<rgba 0 1 1 1>}
     pressedPadding=<ltrb 0 0 0 0>
     pressedSlateSound {resourceObject=null}
    }
   }
   slot {class=BorderSlot, properties {padding=<ltrb 0 0 0 0>}}
  }
  GridPanel_1 {
   widgetClassPath=&GridPanel
   parent=Overlay_0
   widget=GridPanel_1
   slot {class=OverlaySlot, properties {horizontalAlignment=HAlign_Center, verticalAlignment=VAlign_Center}}
  }
  VerticalBox_0 {
   widgetClassPath=&VerticalBox
   parent=GridPanel_1
   widget=VerticalBox_0
   slot {class=GridSlot, properties {column=1, verticalAlignment=VAlign_Top, padding=<ltrb 60 0 60 25>}}
  }
  Overlay_1 {
   widgetClassPath=&Overlay
   parent=VerticalBox_0
   widget=Overlay_1
   slot {class=VerticalBoxSlot, properties {padding=<ltrb 0 0 0 24>}}
  }
  WBP_UIKit_Backplate {
   widgetClassPath=&WBP_UIKit_Backplate
   bIsVariable=true
   parent=Overlay_1
   widget=WBP_UIKit_Backplate
   properties {brightness=Obscure, cornerRadius=Interactive Component}
   slot {class=OverlaySlot, properties {horizontalAlignment=HAlign_Fill, verticalAlignment=VAlign_Fill}}
  }
  HorizontalBox_2 {
   widgetClassPath=&HorizontalBox
   parent=Overlay_1
   widget=HorizontalBox_2
   slot {class=OverlaySlot, properties {padding=<ltrb 25 12 0 12>}}
  }
  Image_TravelIcon {
   widgetClassPath=&Image
   bIsVariable=true
   parent=HorizontalBox_2
   widget=Image_TravelIcon
   properties {brush {
     bIsDynamicallyLoaded=false
     drawAs=Image
     imageType=FullColor
     mirroring=NoMirror
     resourceName=None
     resourceObject=/FortPlaylistTravelUI/Art/M_UI_FortPlaylistTravel_Travel
     tiling=NoTile
     imageSize=<28 28>
     margin=<ltrb 0 0 0 0>
     outlineSettings {
      bUseBrushTransparency=false
      roundingType=HalfHeightRadius
      width=0
      color {colorUseRule=UseColor_Specified, specifiedColor=<rgba 0 0 0 0>}
      cornerRadii=<0 0 0 0>
     }
     tintColor {colorUseRule=UseColor_Specified, specifiedColor=<rgba 1 1 1 1>}
     uVRegion {bIsValid=false, max=<0 0>, min=<0 0>}
    }}
   slot {class=HorizontalBoxSlot, properties {horizontalAlignment=HAlign_Left, verticalAlignment=VAlign_Center}}
  }
  Text_Header {
   widgetClassPath=&CommonTextBlock
   parent=HorizontalBox_2
   widget=Text_Header
   properties {
    justification=Center
    text=Experience Travel
    font {
     bForceMonospaced=false
     bMaterialIsStencil=false
     fontMaterial=null
     fontObject=&HeadingNow
     letterSpacing=0
     monospacedWidth=1
     size=18
     skewAmount=0
     typefaceFontName=75 Medium
     outlineSettings {
      bApplyOutlineToDropShadows=false
      bMiteredCorners=false
      bSeparateFillAlpha=false
      outlineMaterial=null
      outlineSize=0
      outlineColor=<rgba 0 0 0 1>
     }
    }
    shadowOffset=<0 0>
   }
   slot {class=HorizontalBoxSlot, properties {horizontalAlignment=HAlign_Center, padding=<ltrb 15 0 0 0>, size {sizeRule=Fill, value=1}}}
  }
  Overlay_2 {widgetClassPath=&Overlay, parent=VerticalBox_0, widget=Overlay_2, slot {class=VerticalBoxSlot}}
  ContentCarousel {
   widgetClassPath=/Script/CommonUI.CommonWidgetCarousel
   bIsVariable=true
   parent=Overlay_2
   widget=ContentCarousel
   properties {moveSpeed=3}
   slot {class=OverlaySlot}
  }
  WBP_LFNGameTravel_Tile {
   widgetClassPath=&/WBP_FortPlaylistTravel_Tile
   bIsVariable=true
   parent=ContentCarousel
   widget=WBP_LFNGameTravel_Tile
   slot {class=PanelSlot}
  }
  ModalOverlay {
   widgetClassPath=&WBP_UIKit_Backplate
   bIsVariable=true
   parent=Overlay_2
   widget=ModalOverlay
   properties {brightness=Dark, cornerRadius=Interactive Component, renderOpacity=0, visibility=Collapsed, backplateSize=<912 0>}
   slot {class=OverlaySlot, properties {verticalAlignment=VAlign_Fill}}
  }
  ContentCarousel_Nav {
   widgetClassPath=/Script/CommonUI.CommonWidgetCarouselNavBar
   bIsVariable=true
   parent=VerticalBox_0
   widget=ContentCarousel_Nav
   properties {
    buttonWidgetType=&/WBP_FortPlaylistTravel_PipButton.WBP_FortPlaylistTravel_PipButton_C
    buttonPadding=<ltrb 4 4 4 4>
   }
   slot {class=VerticalBoxSlot, properties {horizontalAlignment=HAlign_Center, padding=<ltrb 0 10 0 0>}}
  }
  Button_PrevPage {
   widgetClassPath=&WBP_UIKit_Button_Generic
   bIsVariable=true
   parent=GridPanel_1
   widget=Button_PrevPage
   properties {
    bIsFocusable=false
    promptHorizontalAlignment=HAlign_Center
    promptVerticalAlignment=VAlign_Bottom
    text="<"
    font {
     bForceMonospaced=false
     bMaterialIsStencil=false
     fontMaterial=/FortUILibrary/UIKit/Buttons/MaterialInstances/CTAButtons/MI_UIKit_TextQuiet
     fontObject=&HeadingNow
     letterSpacing=48
     monospacedWidth=1
     size=48
     skewAmount=0
     typefaceFontName=86 Bold
     outlineSettings {
      bApplyOutlineToDropShadows=false
      bMiteredCorners=false
      bSeparateFillAlpha=false
      outlineMaterial=null
      outlineSize=0
      outlineColor=<rgba 0 0 0 1>
     }
    }
    promptNudge=<0 55>
    triggeringInputAction {dataTable=&FortPlaylistTravelUIActions, rowName=PrevPage}
   }
   slot {class=GridSlot, properties {horizontalAlignment=HAlign_Center, rowSpan=3, verticalAlignment=VAlign_Center}}
  }
  Button_NextPage {
   widgetClassPath=&WBP_UIKit_Button_Generic
   bIsVariable=true
   parent=GridPanel_1
   widget=Button_NextPage
   properties {
    bIsFocusable=false
    promptHorizontalAlignment=HAlign_Center
    promptVerticalAlignment=VAlign_Bottom
    text=">"
    font {
     bForceMonospaced=false
     bMaterialIsStencil=false
     fontMaterial=&MI_UIKit_TextQuiet
     fontObject=&HeadingNow
     letterSpacing=48
     monospacedWidth=1
     size=48
     skewAmount=0
     typefaceFontName=86 Bold
     outlineSettings {
      bApplyOutlineToDropShadows=false
      bMiteredCorners=false
      bSeparateFillAlpha=false
      outlineMaterial=null
      outlineSize=0
      outlineColor=<rgba 0 0 0 1>
     }
    }
    promptNudge=<0 55>
    triggeringInputAction {dataTable=&FortPlaylistTravelUIActions, rowName=NextPage}
   }
   slot {class=GridSlot, properties {column=2, horizontalAlignment=HAlign_Center, rowSpan=3, verticalAlignment=VAlign_Center}}
  }
  SafeZone_0 {
   widgetClassPath=&SafeZone
   bIsVariable=true
   parent=Overlay_0
   widget=SafeZone_0
   properties {padLeft=false, padTop=false}
   slot {class=OverlaySlot, properties {horizontalAlignment=HAlign_Right, verticalAlignment=VAlign_Bottom, padding=<ltrb 0 0 80 48>}}
  }
  Button_Close {
   widgetClassPath=/FortUILibrary/UIKit/Buttons/WBP_UIKit_Button_Quiet
   bIsVariable=true
   parent=SafeZone_0
   widget=Button_Close
   properties {bIsFocusable=false, text=Close, triggeringInputAction {dataTable=&FortPlaylistTravelUIActions, rowName=Close}}
   slot {class=SafeZoneSlot}
  }
 ]
}
